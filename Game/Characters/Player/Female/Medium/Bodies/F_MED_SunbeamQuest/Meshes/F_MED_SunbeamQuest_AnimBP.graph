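# Post-process style AnimBlueprint for the F_MED_SunbeamQuest player body, adding secondary motion to tufts, loops, a flap and sleeves.
# Parent mesh pose passes through a deform Control Rig, then chained bone-relative AnimDynamics springs.
# SkydiveAlpha slightly weakens tuft and flap simulation while skydiving.
# Each update also feeds arm-bone motion into cloth WPO material parameters (lowerarms and upperarm_r share region 2), enabling wind during skydive floating.
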AnimBlueprint /Game/Characters/Player/Female/Medium/Bodies/F_MED_SunbeamQuest/Meshes/F_MED_SunbeamQuest_AnimBP
target_skeleton=&/F_MED_SunbeamQuest_Skeleton
graphs [
 AnimGraph {nodes [
   n1 = Control Rig(Source=Copy Pose From Mesh() {bUseAttachedParent=true}) {
    alphaBoolBlend {blendOption=Linear}
    controlRigAssetReference {blueprintRigClass=&/F_MED_SunbeamQuest_Deform_ControlRig.F_MED_SunbeamQuest_Deform_ControlRig_C}
   }
   n2 = Get SkydiveAlpha() -> (SkydiveAlpha:double)
   n3 = Anim Dynamics - Bone: dyn_tuft_A(ComponentPose=Local To Component(LocalPose=@n1), Alpha=@n2, ExternalForce="0,0,0") {
    angularDampingOverride=0.9999
    angularSpringConstant=2000
    bAngularSpring=true
    bOverrideAngularDamping=true
    bUseGravityOverride=true
    simulationSpace=BoneRelative
    alphaBoolBlend {blendOption=Linear}
    boundBone {boneName=dyn_tuft_A}
    componentAppliedLinearAccClamp=<5000 5000 5000>
    componentLinearAccScale=<2 2 1.7>
    componentLinearVelScale=<1.7 1.7 1.7>
    gravityOverride {z=-980}
    physicsBodyDefinitions [{
      sphereCollisionRadius=0
      boundBone {boneName=dyn_tuft_A}
      boxExtents=<17 17 17>
      constraintSetup {angularLimitsMax=<2.5 30 60>, angularLimitsMin=<-2.5 -30 -45>, angularTarget=<1 0 0>}
      localJointOffset=<0.5 0 0>
     }]
    relativeSpaceBone {boneName=root}
   }
   n4 = Anim Dynamics - Bone: dyn_tuft_B(ComponentPose=@n3, Alpha=@n2, ExternalForce="0,0,0") {
    angularDampingOverride=0.9999
    angularSpringConstant=1800
    bAngularSpring=true
    bOverrideAngularDamping=true
    bUseGravityOverride=true
    simulationSpace=BoneRelative
    alphaBoolBlend {blendOption=Linear}
    boundBone {boneName=dyn_tuft_B}
    componentAppliedLinearAccClamp=<5000 5000 5000>
    componentLinearAccScale=<2 2 1.7>
    componentLinearVelScale=<1.7 1.7 1.7>
    gravityOverride {z=-980}
    physicsBodyDefinitions [{
      sphereCollisionRadius=0
      boundBone {boneName=dyn_tuft_B}
      boxExtents=<18 18 18>
      constraintSetup {angularLimitsMax=<2.5 40 45>, angularLimitsMin=<-2.5 -30 -45>, angularTarget=<1 0 0>}
      localJointOffset=<0.65 0 0>
     }]
    relativeSpaceBone {boneName=root}
   }
   n5 = Anim Dynamics - Bone: dyn_tuft_C(ComponentPose=@n4, Alpha=@n2, ExternalForce="0,0,0") {
    angularDampingOverride=0.9999
    angularSpringConstant=2100
    bAngularSpring=true
    bOverrideAngularDamping=true
    bUseGravityOverride=true
    simulationSpace=BoneRelative
    alphaBoolBlend {blendOption=Linear}
    boundBone {boneName=dyn_tuft_C}
    componentAppliedLinearAccClamp=<5000 5000 5000>
    componentLinearAccScale=<2 2 1.7>
    componentLinearVelScale=<1.7 1.7 1.7>
    gravityOverride {z=-980}
    physicsBodyDefinitions [{
      sphereCollisionRadius=0
      boundBone {boneName=dyn_tuft_C}
      boxExtents=<16 16 16>
      constraintSetup {angularLimitsMax=<2.5 5 60>, angularLimitsMin=<-2.5 -30 -30>, angularTarget=<1 0 0>}
      localJointOffset=<0.7 0 0>
     }]
    relativeSpaceBone {boneName=root}
   }
   n6 = Anim Dynamics - Bone: dyn_tuft_D(ComponentPose=@n5, Alpha=@n2, ExternalForce="0,0,0") {
    angularDampingOverride=0.9999
    angularSpringConstant=2000
    bAngularSpring=true
    bOverrideAngularDamping=true
    bUseGravityOverride=true
    simulationSpace=BoneRelative
    alphaBoolBlend {blendOption=Linear}
    boundBone {boneName=dyn_tuft_D}
    componentAppliedLinearAccClamp=<5000 5000 5000>
    componentLinearAccScale=<2 2 1.7>
    componentLinearVelScale=<1.7 1.7 1.7>
    gravityOverride {z=-980}
    physicsBodyDefinitions [{
      sphereCollisionRadius=0
      boundBone {boneName=dyn_tuft_D}
      boxExtents=<20 20 20>
      constraintSetup {angularLimitsMax=<20 10 60>, angularLimitsMin=<-2.5 -20 -30>, angularTarget=<1 0 0>}
      localJointOffset=<0.5 0 0>
     }]
    relativeSpaceBone {boneName=root}
   }
   n7 = Anim Dynamics - Bone: dyn_tuft_E(ComponentPose=@n6, Alpha=@n2, ExternalForce="0,0,0") {
    angularDampingOverride=0.9999
    angularSpringConstant=1800
    bAngularSpring=true
    bOverrideAngularDamping=true
    bUseGravityOverride=true
    simulationSpace=BoneRelative
    alphaBoolBlend {blendOption=Linear}
    boundBone {boneName=dyn_tuft_E}
    componentAppliedLinearAccClamp=<5000 5000 5000>
    componentLinearAccScale=<2 2 1.7>
    componentLinearVelScale=<1.7 1.7 1.7>
    gravityOverride {z=-980}
    physicsBodyDefinitions [{
      sphereCollisionRadius=0
      boundBone {boneName=dyn_tuft_E}
      boxExtents=<14 14 14>
      constraintSetup {angularLimitsMax=<2.5 40 60>, angularLimitsMin=<-2.5 -30 -40>, angularTarget=<1 0 0>}
      localJointOffset=<0.4 0 0>
     }]
    relativeSpaceBone {boneName=root}
   }
   n8 = Anim Dynamics - Bone: dyn_tuft_F(ComponentPose=@n7, Alpha=@n2, ExternalForce="0,0,0") {
    angularDampingOverride=0.9999
    angularSpringConstant=2100
    bAngularSpring=true
    bOverrideAngularDamping=true
    bUseGravityOverride=true
    simulationSpace=BoneRelative
    alphaBoolBlend {blendOption=Linear}
    boundBone {boneName=dyn_tuft_F}
    componentAppliedLinearAccClamp=<5000 5000 5000>
    componentLinearAccScale=<2 2 1.7>
    componentLinearVelScale=<1.7 1.7 1.7>
    gravityOverride {z=-980}
    physicsBodyDefinitions [{
      sphereCollisionRadius=0
      boundBone {boneName=dyn_tuft_F}
      boxExtents=<20 20 20>
      constraintSetup {angularLimitsMax=<2.5 40 60>, angularLimitsMin=<-2.5 -5 -40>, angularTarget=<1 0 0>}
      localJointOffset=<0.5 0 0>
     }]
    relativeSpaceBone {boneName=root}
   }
   n9 = Anim Dynamics - Bone: dyn_tuft_G(ComponentPose=@n8, Alpha=@n2, ExternalForce="0,0,0") {
    angularDampingOverride=0.9999
    angularSpringConstant=2000
    bAngularSpring=true
    bOverrideAngularDamping=true
    bUseGravityOverride=true
    simulationSpace=BoneRelative
    alphaBoolBlend {blendOption=Linear}
    boundBone {boneName=dyn_tuft_G}
    componentAppliedLinearAccClamp=<5000 5000 5000>
    componentLinearAccScale=<2 2 1.7>
    componentLinearVelScale=<1.7 1.7 1.7>
    gravityOverride {z=-980}
    physicsBodyDefinitions [{
      sphereCollisionRadius=0
      boundBone {boneName=dyn_tuft_G}
      boxExtents=<18 18 18>
      constraintSetup {angularLimitsMax=<2.5 20 60>, angularLimitsMin=<-2.5 -20 -40>, angularTarget=<1 0 0>}
      localJointOffset=<0.5 0 0>
     }]
    relativeSpaceBone {boneName=root}
   }
   n10 = Anim Dynamics - Bone: dyn_loop_01_l(ComponentPose=@n9, Alpha=1, ExternalForce="0,0,0") {
    angularDampingOverride=0.9999
    angularSpringConstant=2500
    bAngularSpring=true
    bOverrideAngularDamping=true
    bUseGravityOverride=true
    simulationSpace=BoneRelative
    alphaBoolBlend {blendOption=Linear}
    boundBone {boneName=dyn_loop_01_l}
    componentAppliedLinearAccClamp=<5000 5000 5000>
    componentLinearAccScale=<1.2 1.2 1.2>
    componentLinearVelScale=<1.2 1.2 1.2>
    gravityOverride {z=-980}
    physicsBodyDefinitions [{
      sphereCollisionRadius=0
      boundBone {boneName=dyn_loop_01_l}
      boxExtents=<14 14 14>
      constraintSetup {angularLimitsMax=<2.5 2.5 24>, angularLimitsMin=<-2.5 -2.5 -24>, angularTarget=<1 0 0>}
      localJointOffset=<0.5 0 0>
     }]
    relativeSpaceBone {boneName=root}
   }
   n11 = Anim Dynamics - Bone: dyn_loop_02_l(ComponentPose=@n10, Alpha=1, ExternalForce="0,0,0") {
    angularDampingOverride=0.9999
    angularSpringConstant=2500
    bAngularSpring=true
    bOverrideAngularDamping=true
    bUseGravityOverride=true
    simulationSpace=BoneRelative
    alphaBoolBlend {blendOption=Linear}
    boundBone {boneName=dyn_loop_02_l}
    componentAppliedLinearAccClamp=<5000 5000 5000>
    componentLinearAccScale=<1.2 1.2 1.2>
    componentLinearVelScale=<1.2 1.2 1.2>
    gravityOverride {z=-980}
    physicsBodyDefinitions [{
      sphereCollisionRadius=0
      boundBone {boneName=dyn_loop_02_l}
      boxExtents=<18 18 18>
      constraintSetup {angularLimitsMax=<2.5 2.5 26>, angularLimitsMin=<-2.5 -2.5 -26>, angularTarget=<1 0 0>}
      localJointOffset=<0.65 0 0>
     }]
    relativeSpaceBone {boneName=root}
   }
   n12 = Anim Dynamics - Bone: dyn_loop_01_r(ComponentPose=@n11, Alpha=1, ExternalForce="0,0,0") {
    angularDampingOverride=0.9999
    angularSpringConstant=2500
    bAngularSpring=true
    bOverrideAngularDamping=true
    bUseGravityOverride=true
    simulationSpace=BoneRelative
    alphaBoolBlend {blendOption=Linear}
    boundBone {boneName=dyn_loop_01_r}
    componentAppliedLinearAccClamp=<5000 5000 5000>
    componentLinearAccScale=<1.2 1.2 1.2>
    componentLinearVelScale=<1.2 1.2 1.2>
    gravityOverride {z=-980}
    physicsBodyDefinitions [{
      sphereCollisionRadius=0
      boundBone {boneName=dyn_loop_01_r}
      boxExtents=<15 15 15>
      constraintSetup {angularLimitsMax=<2.5 2.5 25>, angularLimitsMin=<-2.5 -2.5 -25>, angularTarget=<1 0 0>}
      localJointOffset=<-0.5 0 0>
     }]
    relativeSpaceBone {boneName=root}
   }
   n13 = Anim Dynamics - Bone: dyn_loop_02_r(ComponentPose=@n12, Alpha=1, ExternalForce="0,0,0") {
    angularDampingOverride=0.9999
    angularSpringConstant=2500
    bAngularSpring=true
    bOverrideAngularDamping=true
    bUseGravityOverride=true
    simulationSpace=BoneRelative
    alphaBoolBlend {blendOption=Linear}
    boundBone {boneName=dyn_loop_02_r}
    componentAppliedLinearAccClamp=<5000 5000 5000>
    componentLinearAccScale=<1.2 1.2 1.2>
    componentLinearVelScale=<1.2 1.2 1.2>
    gravityOverride {z=-980}
    physicsBodyDefinitions [{
      sphereCollisionRadius=0
      boundBone {boneName=dyn_loop_02_r}
      boxExtents=<18 18 18>
      constraintSetup {angularLimitsMax=<2.5 2.5 23>, angularLimitsMin=<-2.5 -2.5 -23>, angularTarget=<1 0 0>}
      localJointOffset=<-0.65 0 0>
     }]
    relativeSpaceBone {boneName=root}
   }
   n14 = Anim Dynamics - Bone: dyn_flap_01_l(ComponentPose=@n13, Alpha=Get SkydiveAlpha(), ExternalForce="0,0,0") {
    angularDampingOverride=0.9999
    angularSpringConstant=2500
    bAngularSpring=true
    bOverrideAngularDamping=true
    bUseGravityOverride=true
    simulationSpace=BoneRelative
    alphaBoolBlend {blendOption=Linear}
    boundBone {boneName=dyn_flap_01_l}
    componentAppliedLinearAccClamp=<5000 5000 5000>
    componentLinearAccScale=<1.2 1.2 1.2>
    componentLinearVelScale=<1.2 1.2 1.2>
    gravityOverride {z=-980}
    physicsBodyDefinitions [{
      sphereCollisionRadius=0
      boundBone {boneName=dyn_flap_01_l}
      boxExtents=<20 20 20>
      constraintSetup {angularLimitsMax=<2.5 2.5 26>, angularLimitsMin=<-2.5 -2.5 -20>, angularTarget=<1 0 0>}
      localJointOffset=<0.5 0 0>
     }]
    relativeSpaceBone {boneName=root}
   }
   n15 = Anim Dynamics - Bone: dyn_sleeve_upr_l(ComponentPose=@n14, Alpha=1, ExternalForce="0,0,0") {
    angularDampingOverride=0.9999
    angularSpringConstant=2500
    bAngularSpring=true
    bLinearSpring=true
    bOverrideAngularDamping=true
    bOverrideLinearDamping=true
    bUseGravityOverride=true
    linearDampingOverride=0.9999
    linearSpringConstant=100
    simulationSpace=BoneRelative
    alphaBoolBlend {blendOption=Linear}
    boundBone {boneName=dyn_sleeve_upr_l}
    componentAppliedLinearAccClamp=<5000 5000 5000>
    componentLinearAccScale=<1.2 1.2 1.2>
    componentLinearVelScale=<1.2 1.2 1.2>
    gravityOverride {z=-980}
    physicsBodyDefinitions [{
      sphereCollisionRadius=0
      boundBone {boneName=dyn_sleeve_upr_l}
      boxExtents=<8 8 8>
      constraintSetup {
       angularLimitsMax=<2.5 30 20>
       angularLimitsMin=<-2.5 -30 -20>
       angularTarget=<1 0 0>
       linearAxesMax=<0.75 0.75 0.75>
       linearAxesMin=<-0.75 -0.75 -0.75>
      }
      localJointOffset=<0.5 0 0>
     }]
    relativeSpaceBone {boneName=root}
   }
   n16 = Anim Dynamics - Bone: dyn_sleeve_upr_r(ComponentPose=@n15, Alpha=1, ExternalForce="0,0,0") {
    angularDampingOverride=0.9999
    angularSpringConstant=2500
    bAngularSpring=true
    bLinearSpring=true
    bOverrideAngularDamping=true
    bOverrideLinearDamping=true
    bUseGravityOverride=true
    linearDampingOverride=0.9999
    linearSpringConstant=100
    simulationSpace=BoneRelative
    alphaBoolBlend {blendOption=Linear}
    boundBone {boneName=dyn_sleeve_upr_r}
    componentAppliedLinearAccClamp=<5000 5000 5000>
    componentLinearAccScale=<1.2 1.2 1.2>
    componentLinearVelScale=<1.2 1.2 1.2>
    gravityOverride {z=-980}
    physicsBodyDefinitions [{
      sphereCollisionRadius=0
      boundBone {boneName=dyn_sleeve_upr_r}
      boxExtents=<8 8 8>
      constraintSetup {angularLimitsMax=<2.5 30 20>, angularLimitsMin=<-2.5 -30 -20>, angularTarget=<1 0 0>, linearAxesMax=<2 2 2>, linearAxesMin=<-2 -2 -2>}
      localJointOffset=<0.7 0 0>
     }]
    relativeSpaceBone {boneName=root}
   }
   n17 = Anim Dynamics - Bone: dyn_sleeve_lwr_01_l(ComponentPose=@n16, Alpha=1, ExternalForce="0,0,0") {
    angularDampingOverride=0.9999
    angularSpringConstant=2000
    bAngularSpring=true
    bLinearSpring=true
    bOverrideAngularDamping=true
    bOverrideLinearDamping=true
    bUseGravityOverride=true
    bUsePlanarLimit=false
    linearDampingOverride=0.9999
    linearSpringConstant=100
    simulationSpace=BoneRelative
    alphaBoolBlend {blendOption=Linear}
    boundBone {boneName=dyn_sleeve_lwr_01_l}
    componentAppliedLinearAccClamp=<100 100 100>
    componentLinearAccScale=<0.1 0.1 0.1>
    componentLinearVelScale=<0.1 0.1 0.1>
    gravityOverride {z=-980}
    physicsBodyDefinitions [{
      sphereCollisionRadius=0
      boundBone {boneName=dyn_sleeve_lwr_01_l}
      boxExtents=<80 80 80>
      constraintSetup {angularLimitsMax=<0 0 0>, angularLimitsMin=<0 0 0>, angularTarget=<1 0 0>, linearAxesMax=<2 2 2>, linearAxesMin=<-2 -2 -2>}
      localJointOffset=<0.1 0 0>
     }]
    relativeSpaceBone {boneName=root}
   }
   n18 = Anim Dynamics - Bone: dyn_sleeve_lwr_02_l(ComponentPose=@n17, Alpha=1, ExternalForce="0,0,0") {
    angularDampingOverride=0.9999
    angularSpringConstant=2000
    bAngularSpring=true
    bLinearSpring=true
    bOverrideAngularDamping=true
    bOverrideLinearDamping=true
    bUseGravityOverride=true
    bUsePlanarLimit=false
    linearDampingOverride=0.9999
    linearSpringConstant=100
    simulationSpace=BoneRelative
    alphaBoolBlend {blendOption=Linear}
    boundBone {boneName=dyn_sleeve_lwr_02_l}
    componentAppliedLinearAccClamp=<100 100 100>
    componentLinearAccScale=<0.1 0.1 0.1>
    componentLinearVelScale=<0.1 0.1 0.1>
    gravityOverride {z=-980}
    physicsBodyDefinitions [{
      sphereCollisionRadius=0
      boundBone {boneName=dyn_sleeve_lwr_02_l}
      boxExtents=<80 80 80>
      constraintSetup {angularLimitsMax=<0 0 0>, angularLimitsMin=<0 0 0>, angularTarget=<1 0 0>, linearAxesMax=<2 2 2>, linearAxesMin=<-2 -2 -2>}
      localJointOffset=<0.1 0 0>
     }]
    relativeSpaceBone {boneName=root}
   }
   n19 = Anim Dynamics - Bone: dyn_sleeve_lwr_01_r(ComponentPose=@n18, Alpha=1, ExternalForce="0,0,0") {
    angularDampingOverride=0.9999
    angularSpringConstant=2000
    bAngularSpring=true
    bLinearSpring=true
    bOverrideAngularDamping=true
    bOverrideLinearDamping=true
    bUseGravityOverride=true
    bUsePlanarLimit=false
    linearDampingOverride=0.9999
    linearSpringConstant=100
    simulationSpace=BoneRelative
    alphaBoolBlend {blendOption=Linear}
    boundBone {boneName=dyn_sleeve_lwr_01_r}
    componentAppliedLinearAccClamp=<100 100 100>
    componentLinearAccScale=<0.1 0.1 0.1>
    componentLinearVelScale=<0.1 0.1 0.1>
    gravityOverride {z=-980}
    physicsBodyDefinitions [{
      sphereCollisionRadius=0
      boundBone {boneName=dyn_sleeve_lwr_01_r}
      boxExtents=<80 80 80>
      constraintSetup {angularLimitsMax=<0 0 0>, angularLimitsMin=<0 0 0>, angularTarget=<1 0 0>, linearAxesMax=<2 2 2>, linearAxesMin=<-2 -2 -2>}
      localJointOffset=<0.1 0 0>
     }]
    relativeSpaceBone {boneName=root}
   }
   n20 = Anim Dynamics - Bone: dyn_sleeve_lwr_02_r(ComponentPose=@n19, Alpha=1, ExternalForce="0,0,0") {
    angularDampingOverride=0.9999
    angularSpringConstant=2000
    bAngularSpring=true
    bLinearSpring=true
    bOverrideAngularDamping=true
    bOverrideLinearDamping=true
    bUseGravityOverride=true
    bUsePlanarLimit=false
    linearDampingOverride=0.9999
    linearSpringConstant=100
    simulationSpace=BoneRelative
    alphaBoolBlend {blendOption=Linear}
    boundBone {boneName=dyn_sleeve_lwr_02_r}
    componentAppliedLinearAccClamp=<100 100 100>
    componentLinearAccScale=<0.1 0.1 0.1>
    componentLinearVelScale=<0.1 0.1 0.1>
    gravityOverride {z=-980}
    physicsBodyDefinitions [{
      sphereCollisionRadius=0
      boundBone {boneName=dyn_sleeve_lwr_02_r}
      boxExtents=<80 80 80>
      constraintSetup {angularLimitsMax=<0 0 0>, angularLimitsMin=<0 0 0>, angularTarget=<1 0 0>, linearAxesMax=<2 2 2>, linearAxesMin=<-2 -2 -2>}
      localJointOffset=<0.1 0 0>
     }]
    relativeSpaceBone {boneName=root}
   }
   Output Pose(Result=Component To Local(ComponentPose=@n20)) {layerGroup=None}
  ]}
 EventGraph {nodes [
   n1 = Event BlueprintUpdateAnimation()
   > n2 = Sequence()
   > Is Valid(InputObject=Get playerPawn() -> (playerPawn:Fort Player Pawn Athena Object Reference))
   > Set WPO Globals Function(
    Material=Get matInstanceDynamic() -> (matInstanceDynamic:Material Instance Dynamic Object Reference)
    Overlay Material=Get overlayMID() -> (overlayMID:Material Instance Dynamic Object Reference)
    Skeletal Mesh=Get skelMesh() -> (skelMesh:Skeletal Mesh Component Object Reference)
    "Local Noise Transform  Location"="Get Local Noise Transform  Location"() -> ("Local Noise Transform  Location":Vector)
    Local Noise Offset=Get Local Noise Offset()
   )
   > Set WPO Regional Function!(
    execute
    Material:Material Instance Dynamic Object Reference=Get matInstanceDynamic()
    Overlay Material:Material Instance Dynamic Object Reference=Get overlayMID()
    Skeletal Mesh:Skeletal Mesh Component Object Reference=Get skelMesh()
    Velocity Transform:Name=upperarm_l
    Velocity Interp Speed:double=10
    Collision Transform:Name=upperarm_l
    Collision Transform Axis:EAxis Enum
    Invert Collision Transform Axis:bool=false
    Delta Time X:double=@n1.DeltaTimeX
    Previous Velocity Transform Location:"Vector (by ref)"=Get R1 Velocity Transform Location()
    Previous Velocity Transform Velocity:"Vector (by ref)"=Get R1 Previous Velocity Transform Velocity() -> (R1 Previous Velocity Transform Velocity:Vector)
    Region:int=1
   ) -> (then)
   > Set WPO Regional Function(
    Material=Get matInstanceDynamic()
    Overlay Material=Get overlayMID()
    Skeletal Mesh=Get skelMesh()
    Velocity Transform=upperarm_r
    Velocity Interp Speed=10
    Collision Transform=upperarm_r
    Invert Collision Transform Axis=false
    Delta Time X=@n1.DeltaTimeX
    Previous Velocity Transform Location=Get R2 Velocity Transform Location()
    Previous Velocity Transform Velocity=Get R2 Velocity Transform Velocity()
    Region=2
   )
   > Set WPO Regional Function(
    Material=Get matInstanceDynamic()
    Overlay Material=Get overlayMID()
    Skeletal Mesh=Get skelMesh()
    Velocity Transform=lowerarm_l
    Velocity Interp Speed=10
    Collision Transform=lowerarm_l
    Invert Collision Transform Axis=false
    Delta Time X=@n1.DeltaTimeX
    Previous Velocity Transform Location=Get R3 Velocity Transform Location() -> (R3 Velocity Transform Location:Vector)
    Previous Velocity Transform Velocity=Get R3 Velocity Transform Velocity() -> (R3 Velocity Transform Velocity:Vector)
    Region=2
   )
   > Set WPO Regional Function(
    Material=Get matInstanceDynamic()
    Overlay Material=Get overlayMID()
    Skeletal Mesh=Get skelMesh()
    Velocity Transform=lowerarm_r
    Velocity Interp Speed=10
    Collision Transform=lowerarm_r
    Invert Collision Transform Axis=false
    Delta Time X=@n1.DeltaTimeX
    Previous Velocity Transform Location=Get R4 Velocity Transform Location() -> (R4 Velocity Transform Location:Vector)
    Previous Velocity Transform Velocity=Get R4 Velocity Transform Velocity() -> (R4 Velocity Transform Velocity:Vector)
    Region=2
   )
   > n3 = Branch(Condition=Get bIsSkydiveFloating())
   > n4 = Set WPO Pho Wind Intensity(WPO Pho Wind Intensity=1)
   SetScalarParameterValue(execute=[@n5.then, @n4.then], self=Get matInstanceDynamic(), ParameterName=ClothSim_WindStrength, Value=Get WPO Pho Wind Intensity())
   n5 = Set WPO Pho Wind Intensity(execute=@n3.else, WPO Pho Wind Intensity=0)
   n6 = Branch(execute=@n2.then_1, Condition=Get bIsSkyDiving())
   > Set SkydiveAlpha(SkydiveAlpha:double=0.75) -> (Output_Get:double)
   Set SkydiveAlpha(execute=@n6.else, SkydiveAlpha=0.9)
   Event BlueprintInitializeAnimation()
   > Set SkelMesh(SkelMesh=GetOwnerSkeletalMeshComponent())
   n7 = TryGetPawnOwner()
   > Is Valid(InputObject=@n7)
   > n8 = Cast To FortPlayerPawnAthena(Object=@n7)
   > Set playerPawn(playerPawn:Fort Player Pawn Athena Object Reference=@n8) -> (Output_Get:Fort Player Pawn Athena Object Reference)
   Event OnCharacterCustomizationLoaded()
   > n9 = Set skelMesh(skelMesh:Skeletal Mesh Component Object Reference=GetPartSkeletalMeshComponent()) -> (Output_Get:Skeletal Mesh Component Object Reference)
   > n10 = Get Dynamic Material Instance(SkeletalMeshComponent=@n9)
   > Set matInstanceDynamic(matInstanceDynamic:Material Instance Dynamic Object Reference=@n10) -> (Output_Get:Material Instance Dynamic Object Reference)
   > n11 = GetOrCreateDynamicOverlayMaterial(MeshComponent=@n9, MaterialName=None)
   > Set overlayMID(overlayMID:Material Instance Dynamic Object Reference=@n11) -> (Output_Get:Material Instance Dynamic Object Reference)
  ]}
]
variables [
 activeKill {type=boolean}
 cape {type=boolean}
 killCount {type=number}
 limitTailRange {type=boolean}
 Local Noise Offset {type=Vector}
 "Local Noise Transform  Location" {type=Vector}
 matInstanceDynamic {type=MaterialInstanceDynamic}
 overlayMID {type=MaterialInstanceDynamic}
 playerPawn {type=FortPlayerPawnAthena}
 R1 Previous Velocity Transform Velocity {type=Vector}
 R1 Velocity Transform Location {type=Vector}
 R2 Velocity Transform Location {type=Vector}
 R2 Velocity Transform Velocity {type=Vector}
 R3 Velocity Transform Location {type=Vector}
 R3 Velocity Transform Velocity {type=Vector}
 R4 Velocity Transform Location {type=Vector}
 R4 Velocity Transform Velocity {type=Vector}
 skelMesh {type=SkeletalMeshComponent}
 SkydiveAlpha {type=number}
 Tail_1_Scale {type=Vector}
 Tail_2_Scale {type=Vector}
 Tail_3_Scale {type=Vector}
 Tail_4_Scale {type=Vector}
 Tail_5_Scale {type=Vector}
 tail_rotate {type=Rotator}
 tail_translate {type=Vector}
 TailPrime_1_Scale {type=Vector}
 TailPrime_2_Scale {type=Vector}
 tailSkydiveAlpha {type=number}
 useTailCtrlRig {type=boolean}
 WPO Pho Wind Intensity {type=number}
]
