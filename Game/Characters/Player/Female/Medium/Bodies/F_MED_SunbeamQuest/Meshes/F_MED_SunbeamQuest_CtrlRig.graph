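# Control Rig for the SunbeamQuest medium female player body that procedurally animates seven tail bone chains (A to G) hung from the pelvis.
# A sequence first applies fixed rotation tweaks, plus small shifts on B and E, to each tail's parent bone.
# It then runs Chain Harmonics from each master bone, combining reach, zero-gravity pendulum and per-tail wave settings.
# Other dyn bones are untouched, and debug drawing is enabled on every chain.
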ControlRigBlueprint /Game/Characters/Player/Female/Medium/Bodies/F_MED_SunbeamQuest/Meshes/F_MED_SunbeamQuest_CtrlRig
graphs [
 ContainedGraph {nodes [
   n1 = RigVMFunctionEntryNode()
   SequenceAggregate = &RigVMFunction_Sequence(ExecuteContext=@n1.ExecuteContext)
   SequenceAggregate_1 = &RigVMFunction_Sequence(ExecuteContext=@SequenceAggregate.B)
   RigVMFunctionReturnNode(A=@SequenceAggregate.A, B=@SequenceAggregate_1.A, C=@SequenceAggregate_1.B)
  ]}
 RigVMFunctionLibrary {links [], nodes []}
 RigVMModel {nodes [
   RigUnit_BeginExecution = &RigUnit_BeginExecution()
   SequenceAggregate = RigVMAggregateNode(ExecuteContext=@RigUnit_BeginExecution.ExecutePin)
   &RigUnit_ModifyTransforms(
    ItemToModify=((Item=(Type=Bone,Name="dyn_tail_A_parent"),Transform=(Rotation=(X=0,Y=0,Z=0,W=1),Translation=(X=0,Y=0,Z=0),Scale3D=(X=1,Y=1,Z=1))))
    Weight=1
    ItemToModify.0.Transform.Rotation=&RigVMFunction_MathQuaternionFromEuler(Euler=(X=0,Y=2,Z=25), RotationOrder=XYZ).Result
    ExecutePin=@SequenceAggregate.A
   )
   > &RigUnit_ModifyTransforms(
    ItemToModify=((Item=(Type=Bone,Name="dyn_tail_B_parent"),Transform=(Rotation=(X=0,Y=0,Z=0,W=1),Translation=(X=0,Y=0,Z=-2),Scale3D=(X=1,Y=1,Z=1))))
    Weight=1
    ItemToModify.0.Transform.Rotation=&RigVMFunction_MathQuaternionFromEuler(Euler=(X=0,Y=-2,Z=6), RotationOrder=XYZ).Result
   )
   > &RigUnit_ModifyTransforms(
    ItemToModify=((Item=(Type=Bone,Name="dyn_tail_C_parent"),Transform=(Rotation=(X=0,Y=0,Z=0,W=1),Translation=(X=0,Y=0,Z=0),Scale3D=(X=1,Y=1,Z=1))))
    Weight=1
    ItemToModify.0.Transform.Rotation=&RigVMFunction_MathQuaternionFromEuler(Euler=(X=0,Y=-2,Z=-8), RotationOrder=XYZ).Result
   )
   > &RigUnit_ModifyTransforms(
    ItemToModify=((Item=(Type=Bone,Name="dyn_tail_D_parent"),Transform=(Rotation=(X=0,Y=0,Z=0,W=1),Translation=(X=0,Y=0,Z=0),Scale3D=(X=1,Y=1,Z=1))))
    Weight=1
    ItemToModify.0.Transform.Rotation=&RigVMFunction_MathQuaternionFromEuler(Euler=(X=0,Y=4,Z=0), RotationOrder=XYZ).Result
   )
   > &RigUnit_ModifyTransforms(
    ItemToModify=((Item=(Type=Bone,Name="dyn_tail_E_parent"),Transform=(Rotation=(X=0,Y=0,Z=0,W=1),Translation=(X=0,Y=0,Z=2),Scale3D=(X=1,Y=1,Z=1))))
    Weight=1
    ItemToModify.0.Transform.Rotation=&RigVMFunction_MathQuaternionFromEuler(Euler=(X=0,Y=18,Z=8), RotationOrder=XYZ).Result
   )
   > &RigUnit_ModifyTransforms(
    ItemToModify=((Item=(Type=Bone,Name="dyn_tail_G_parent"),Transform=(Rotation=(X=0,Y=0,Z=0,W=1),Translation=(X=0,Y=0,Z=0),Scale3D=(X=1,Y=1,Z=1))))
    Weight=1
    ItemToModify.0.Transform.Rotation=&RigVMFunction_MathQuaternionFromEuler(Euler=(X=0,Y=-4,Z=30), RotationOrder=XYZ).Result
   )
   > &RigUnit_ModifyTransforms(
    ItemToModify=((Item=(Type=Bone,Name="dyn_tail_F_parent"),Transform=(Rotation=(X=0,Y=0,Z=0,W=1),Translation=(X=0,Y=0,Z=0),Scale3D=(X=1,Y=1,Z=1))))
    Weight=1
    ItemToModify.0.Transform.Rotation=&RigVMFunction_MathQuaternionFromEuler(Euler=(X=0,Y=2,Z=16), RotationOrder=XYZ).Result
   )
   /Script/ControlRig.RigUnit_ChainHarmonicsPerItem(
    ChainRoot=(Type=Bone,Name="dyn_mstr_tail_A")
    DrawWorldOffset=(Rotation=(X=0,Y=0,Z=0,W=1),Translation=(X=0,Y=0,Z=0),Scale3D=(X=1,Y=1,Z=1))
    Pendulum=(bEnabled=True,PendulumStiffness=2,PendulumGravity=(X=0,Y=0,Z=0),PendulumBlend=0.75,PendulumDrag=0.98,PendulumMaximum=0.1,PendulumEase=BounceEaseInOut,UnwindAxis=(X=0,Y=1,Z=0),UnwindMinimum=0.2,UnwindMaximum=0.05)
    Reach=(bEnabled=True,ReachTarget=(X=0,Y=0,Z=0),ReachAxis=(X=1,Y=0,Z=0))
    Speed=(X=-1.3,Y=-1.3,Z=-1.3)
    Wave=(bEnabled=True,WaveFrequency=(X=60,Y=60,Z=60),WaveAmplitude=(X=0,Y=0.15,Z=0.2),WaveOffset=(X=0,Y=1,Z=3),WaveNoise=(X=0,Y=0,Z=0),WaveMaximum=2,WaveEase=Linear)
    WaveCurve=(EditorCurveData=(Keys=((InterpMode=RCIM_Cubic,TangentMode=RCTM_User,ArriveTangent=2.16,LeaveTangent=2.16),(InterpMode=RCIM_Cubic,TangentMode=RCTM_User,Time=1,Value=1))))
    bDrawDebug=True
    ExecutePin=@SequenceAggregate.B
   ) {is=struct RigVMUnitNode}
   > &RigUnit_ChainHarmonicsPerItem(
    ChainRoot=(Type=Bone,Name="dyn_mstr_tail_B")
    DrawWorldOffset=(Rotation=(X=0,Y=0,Z=0,W=1),Translation=(X=0,Y=0,Z=0),Scale3D=(X=1,Y=1,Z=1))
    Pendulum=(bEnabled=True,PendulumStiffness=2,PendulumGravity=(X=0,Y=0,Z=0),PendulumBlend=0.75,PendulumDrag=0.98,PendulumMaximum=0.1,UnwindAxis=(X=0,Y=1,Z=0),UnwindMinimum=0.2,UnwindMaximum=0.05)
    Reach=(bEnabled=True,ReachTarget=(X=0,Y=0,Z=0),ReachAxis=(X=1,Y=0,Z=0))
    Speed=(X=-1.3,Y=-1.3,Z=-1.3)
    Wave=(bEnabled=True,WaveFrequency=(X=60,Y=60,Z=60),WaveAmplitude=(X=0,Y=0.2,Z=0.07),WaveOffset=(X=0,Y=1,Z=2),WaveNoise=(X=0,Y=0,Z=0),WaveMaximum=1.4,WaveEase=BounceEaseInOut)
    WaveCurve=(EditorCurveData=(Keys=((InterpMode=RCIM_Cubic,TangentMode=RCTM_User,ArriveTangent=2.1057,LeaveTangent=2.1057),(InterpMode=RCIM_Cubic,TangentMode=RCTM_User,Time=1,Value=1))))
    bDrawDebug=True
   )
   > &RigUnit_ChainHarmonicsPerItem(
    ChainRoot=(Type=Bone,Name="dyn_mstr_tail_C")
    DrawWorldOffset=(Rotation=(X=0,Y=0,Z=0,W=1),Translation=(X=0,Y=0,Z=0),Scale3D=(X=1,Y=1,Z=1))
    Pendulum=(bEnabled=True,PendulumStiffness=2,PendulumGravity=(X=0,Y=0,Z=0),PendulumBlend=0.75,PendulumDrag=0.98,PendulumMaximum=0.1,UnwindAxis=(X=0,Y=1,Z=0),UnwindMinimum=0.2,UnwindMaximum=0.05)
    Reach=(bEnabled=True,ReachTarget=(X=0,Y=0,Z=0),ReachAxis=(X=1,Y=0,Z=0))
    Speed=(X=-1.3,Y=-1.3,Z=-1.3)
    Wave=(bEnabled=True,WaveFrequency=(X=60,Y=60,Z=60),WaveAmplitude=(X=0,Y=0.15,Z=0.2),WaveOffset=(X=0,Y=1,Z=2),WaveNoise=(X=0,Y=0,Z=0),WaveMaximum=1.5,WaveEase=Linear)
    WaveCurve=(EditorCurveData=(Keys=((InterpMode=RCIM_Cubic,TangentMode=RCTM_User,ArriveTangent=2.4706,LeaveTangent=2.4706),(InterpMode=RCIM_Cubic,TangentMode=RCTM_User,Time=1,Value=1))))
    bDrawDebug=True
   )
   > &RigUnit_ChainHarmonicsPerItem(
    ChainRoot=(Type=Bone,Name="dyn_mstr_tail_D")
    DrawWorldOffset=(Rotation=(X=0,Y=0,Z=0,W=1),Translation=(X=0,Y=0,Z=0),Scale3D=(X=1,Y=1,Z=1))
    Pendulum=(bEnabled=True,PendulumStiffness=2,PendulumGravity=(X=0,Y=0,Z=0),PendulumBlend=0.75,PendulumDrag=0.98,PendulumMaximum=0.1,PendulumEase=BounceEaseInOut,UnwindAxis=(X=0,Y=1,Z=0),UnwindMinimum=0.2,UnwindMaximum=0.05)
    Reach=(bEnabled=True,ReachTarget=(X=0,Y=0,Z=0),ReachAxis=(X=1,Y=0,Z=0))
    Speed=(X=-1.3,Y=-1.3,Z=-1.3)
    Wave=(bEnabled=True,WaveFrequency=(X=60,Y=60,Z=60),WaveAmplitude=(X=0,Y=-0.2,Z=0.25),WaveOffset=(X=0,Y=1,Z=2),WaveNoise=(X=0,Y=0,Z=0),WaveMaximum=1.3,WaveEase=BounceEaseInOut)
    WaveCurve=(EditorCurveData=(Keys=((InterpMode=RCIM_Cubic,TangentMode=RCTM_User,ArriveTangent=2.6284,LeaveTangent=2.6284),(InterpMode=RCIM_Cubic,TangentMode=RCTM_User,Time=1,Value=1,ArriveTangent=0.0046,LeaveTangent=0.0046))))
    bDrawDebug=True
   )
   > &RigUnit_ChainHarmonicsPerItem(
    ChainRoot=(Type=Bone,Name="dyn_mstr_tail_E")
    DrawWorldOffset=(Rotation=(X=0,Y=0,Z=0,W=1),Translation=(X=0,Y=0,Z=0),Scale3D=(X=1,Y=1,Z=1))
    Pendulum=(bEnabled=True,PendulumStiffness=2,PendulumGravity=(X=0,Y=0,Z=0),PendulumBlend=0.75,PendulumDrag=0.98,PendulumMaximum=0.1,UnwindAxis=(X=0,Y=1,Z=0),UnwindMinimum=0.2,UnwindMaximum=0.05)
    Reach=(bEnabled=True,ReachTarget=(X=0,Y=0,Z=0),ReachAxis=(X=1,Y=0,Z=0))
    Speed=(X=-1.3,Y=-1.3,Z=-1.3)
    Wave=(bEnabled=True,WaveFrequency=(X=60,Y=60,Z=60),WaveAmplitude=(X=0,Y=-0.3,Z=0.2),WaveOffset=(X=0,Y=1,Z=2),WaveNoise=(X=0,Y=0,Z=0),WaveMaximum=2,WaveEase=BounceEaseIn)
    WaveCurve=(EditorCurveData=(Keys=((InterpMode=RCIM_Cubic,TangentMode=RCTM_User,ArriveTangent=2.8375,LeaveTangent=2.8375),(InterpMode=RCIM_Cubic,TangentMode=RCTM_User,Time=1,Value=1))))
    bDrawDebug=True
   )
   > &RigUnit_ChainHarmonicsPerItem(
    ChainRoot=(Type=Bone,Name="dyn_mstr_tail_F")
    DrawWorldOffset=(Rotation=(X=0,Y=0,Z=0,W=1),Translation=(X=0,Y=0,Z=0),Scale3D=(X=1,Y=1,Z=1))
    Pendulum=(bEnabled=True,PendulumStiffness=2,PendulumGravity=(X=0,Y=0,Z=0),PendulumBlend=0.75,PendulumDrag=0.98,PendulumMaximum=0.1,UnwindAxis=(X=0,Y=1,Z=0),UnwindMinimum=0.2,UnwindMaximum=0.05)
    Reach=(bEnabled=True,ReachTarget=(X=0,Y=0,Z=0),ReachAxis=(X=1,Y=0,Z=0))
    Speed=(X=-1.3,Y=-1.3,Z=-1.3)
    Wave=(bEnabled=True,WaveFrequency=(X=60,Y=60,Z=60),WaveAmplitude=(X=0,Y=-0.2,Z=0.2),WaveOffset=(X=0,Y=1,Z=2),WaveNoise=(X=0,Y=0,Z=0),WaveMaximum=1.3,WaveEase=BounceEaseInOut)
    WaveCurve=(EditorCurveData=(Keys=((InterpMode=RCIM_Cubic,TangentMode=RCTM_User,Time=-0.0089,Value=-0.004,ArriveTangent=2.1951,LeaveTangent=2.1951),(InterpMode=RCIM_Cubic,TangentMode=RCTM_User,Time=1,Value=1,ArriveTangent=-0.0162,LeaveTangent=-0.0162))))
    bDrawDebug=True
   )
   > &RigUnit_ChainHarmonicsPerItem(
    ChainRoot=(Type=Bone,Name="dyn_mstr_tail_G")
    DrawWorldOffset=(Rotation=(X=0,Y=0,Z=0,W=1),Translation=(X=0,Y=0,Z=0),Scale3D=(X=1,Y=1,Z=1))
    Pendulum=(bEnabled=True,PendulumStiffness=2,PendulumGravity=(X=0,Y=0,Z=0),PendulumBlend=0.75,PendulumDrag=0.98,PendulumMaximum=0.1,UnwindAxis=(X=0,Y=1,Z=0),UnwindMinimum=0.2,UnwindMaximum=0.05)
    Reach=(bEnabled=True,ReachTarget=(X=0,Y=0,Z=0),ReachAxis=(X=1,Y=0,Z=0))
    Speed=(X=-1.3,Y=-1.3,Z=-1.3)
    Wave=(bEnabled=True,WaveFrequency=(X=60,Y=60,Z=60),WaveAmplitude=(X=0,Y=-0.2,Z=0.15),WaveOffset=(X=0,Y=1,Z=2),WaveNoise=(X=0,Y=0,Z=0),WaveMaximum=1.3,WaveEase=BounceEaseInOut)
    WaveCurve=(EditorCurveData=(Keys=((InterpMode=RCIM_Cubic,TangentMode=RCTM_User,ArriveTangent=2.6527,LeaveTangent=2.6527),(InterpMode=RCIM_Cubic,TangentMode=RCTM_User,Time=1,Value=1))))
    bDrawDebug=True
   )
  ]}
]
hierarchy [
 root {loc=<0 0 0>, rot=<0 0 0 1>}
 pelvis {parent=root, loc=<0 -0.5301 98.0571>, rot=<0.0127 0.707 -0.0127 -0.707>}
 spine_01 {parent=pelvis, loc=<6.1208 -0.9328 0>, rot=<0 0 0.034 -0.9994>}
 spine_02 {parent=spine_01, loc=<6.5996 -0.2095 0>, rot=<0 0 0.0238 -0.9997>}
 spine_03 {parent=spine_02, loc=<6.4914 -0.1237 0>, rot=<0 0 0.0149 -0.9999>}
 spine_04 {parent=spine_03, loc=<6.9598 0.0794 0>, rot=<0 0 -0.0132 -0.9999>}
 spine_05 {parent=spine_04, loc=<6.9987 0 0>, rot=<0 0 -0.0015 -1>}
 attach_cape {parent=spine_05, loc=<9.3216 -8.4501 0>, rot=<0.0616 -0.7044 -0.0616 -0.7044>}
 neck_01 {parent=spine_05, loc=<11.3138 0.1529 0>, rot=<0 0 -0.2003 -0.9797>}
 neck_02 {parent=neck_01, loc=<4.5627 0 0>, rot=<0 0 0.0504 -0.9987>}
 head {parent=neck_02, loc=<5.3914 0 0>, rot=<0 0 0.115 -0.9934>}
 attach_backpack {parent=spine_05, loc=<0 -8.35 0>, rot=<-0.0093 -0.707 0.0093 -0.707>}
 clavicle_l {parent=spine_05, loc=<5.4831 -0.1053 -3.2946>, rot=<0.0465 -0.7289 -0.009 -0.6829>}
 upperarm_l {parent=clavicle_l, loc=<9.8658 -1.2136 -0.4105>, rot=<0.0164 -0.3245 -0.0269 -0.9454>}
 lowerarm_l {parent=upperarm_l, loc=<26.465 -0.8345 0.259>, rot=<0.0036 0.0161 -0.2298 -0.9731>}
 hand_l {parent=lowerarm_l, loc=<22.7355 -0.3203 -0.6105>, rot=<-0.6784 -0.0223 0.0735 -0.7307>}
 thumb_01_l {parent=hand_l, loc=<1.3118 -0.3191 -2.459>, rot=<0.6546 -0.4042 0.0664 -0.6354>}
 thumb_02_l {parent=thumb_01_l, loc=<5.9336 -0.0448 0.0401>, rot=<-0.0037 0.0213 0.298 -0.9543>}
 thumb_03_l {parent=thumb_02_l, loc=<2.2223 0.3682 -0.0907>, rot=<0.0118 -0.013 0.0066 -0.9998>}
 index_metacarpal_l {parent=hand_l, loc=<1.6535 0.0183 -2.96>, rot=<0.0158 -0.0124 -0.0173 -0.9996>}
 index_01_l {parent=index_metacarpal_l, loc=<8.3008 0.0357 -0.0524>, rot=<-0.0044 0.0485 0.0924 -0.9945>}
 index_02_l {parent=index_01_l, loc=<3.3587 -0.3004 0.0452>, rot=<-0.002 0.0027 0.3225 -0.9466>}
 index_03_l {parent=index_02_l, loc=<2.5578 -0.123 -0.0252>, rot=<-0.002 0.0067 0.2558 -0.9667>}
 middle_metacarpal_l {parent=hand_l, loc=<2.1856 -0.1725 -0.5984>, rot=<0.006 0.0168 -0.0329 -0.9993>}
 middle_01_l {parent=middle_metacarpal_l, loc=<7.3275 -0.2447 0.0129>, rot=<-0.0208 0.0042 0.0858 -0.9961>}
 middle_02_l {parent=middle_01_l, loc=<4.0123 0.1401 0.1175>, rot=<-0.0218 0.0258 0.3512 -0.9357>}
 middle_03_l {parent=middle_02_l, loc=<2.6077 0.0534 0.1373>, rot=<0.0082 0.0188 0.1708 -0.9851>}
 ring_metacarpal_l {parent=hand_l, loc=<2.2074 -0.3625 1.3638>, rot=<-0.0284 0.0722 -0.0458 -0.9959>}
 ring_01_l {parent=ring_metacarpal_l, loc=<7.0295 -0.6197 -0.1446>, rot=<-0.0489 -0.0252 0.1892 -0.9804>}
 ring_02_l {parent=ring_01_l, loc=<3.6174 0.4647 -0.1027>, rot=<0 0 0.2806 -0.9598>}
 ring_03_l {parent=ring_02_l, loc=<2.4152 0.114 0.0199>, rot=<0 0 0.1189 -0.9929>}
 pinky_metacarpal_l {parent=hand_l, loc=<1.7616 -0.9753 3.3814>, rot=<-0.0397 0.0814 -0.069 -0.9935>}
 pinky_01_l {parent=pinky_metacarpal_l, loc=<7.1709 -0.4831 -0.1094>, rot=<-0.0654 -0.0035 0.2242 -0.9723>}
 pinky_02_l {parent=pinky_01_l, loc=<2.8274 0.0893 -0.1043>, rot=<0.0037 0.0011 0.2854 -0.9584>}
 pinky_03_l {parent=pinky_02_l, loc=<2.2046 0.0307 -0.0078>, rot=<-0.0002 0.0098 0.1445 -0.9895>}
 deform_wrist_dn_base_l {parent=hand_l, loc=<0 -1.5326 0>, rot=<0 0 0 -1>}
 deform_wrist_dn_l {parent=deform_wrist_dn_base_l, loc=<0 0 0>, rot=<0 0 0 1>}
 deform_wrist_up_base_l {parent=hand_l, loc=<0 2.3009 0>, rot=<0 0 0 -1>}
 deform_wrist_up_l {parent=deform_wrist_up_base_l, loc=<0 0 0>, rot=<0 0 0 1>}
 lowerarm_twist_02_l {parent=lowerarm_l, loc=<10.6707 0 0>, rot=<0 0 0 1>}
 lowerarm_twist_01_l {parent=lowerarm_l, loc=<18.316 0 0>, rot=<0 0 0 1>}
 dyn_sleeve_lwr_01_l {parent=lowerarm_twist_01_l, loc=<-3.906 -2.405 -5.336>, rot=<-0.3017 -0.3664 0.6873 -0.5499>}
 dyn_sleeve_lwr_02_l {parent=lowerarm_twist_01_l, loc=<-4.1536 1.8003 5.2354>, rot=<-0.5405 -0.5638 -0.4572 0.4253>}
 deform_elbow_in_base_l {parent=lowerarm_l, loc=<0.1834 3.6062 -0.2572>, rot=<0 0 0 -1>}
 deform_elbow_in_l {parent=deform_elbow_in_base_l, loc=<0 0 0>, rot=<0 0 0 1>}
 deform_elbow_out_base_l {parent=lowerarm_l, loc=<0 -4 0>, rot=<0 0 0 1>}
 deform_elbow_out_l {parent=deform_elbow_out_base_l, loc=<0 0 0>, rot=<0 0 0 1>}
 upperarm_twist_01_l {parent=upperarm_l, loc=<7.2135 -0.1134 0>, rot=<0 0 0 1>}
 deform_upperarm_bk_base_l {parent=upperarm_twist_01_l, loc=<-6.4053 -4.3746 0.314>, rot=<0 0 0 -1>}
 deform_upperarm_bk_l {parent=deform_upperarm_bk_base_l, loc=<0 0 0>, rot=<0 0 0 1>}
 deform_upperarm_fr_base_l {parent=upperarm_twist_01_l, loc=<-5.6312 5.092 0.3354>, rot=<0 0 0 -1>}
 deform_upperarm_fr_l {parent=deform_upperarm_fr_base_l, loc=<0 0 0>, rot=<0 0 0 1>}
 deform_upperarm_up_base_l {parent=upperarm_twist_01_l, loc=<-8.891 0.4825 3.9517>, rot=<0 0 0 -1>}
 deform_upperarm_up_l {parent=deform_upperarm_up_base_l, loc=<0 0 0>, rot=<0 0 0 1>}
 dyn_sleeve_upr_l {parent=upperarm_twist_01_l, loc=<5.6856 -0.3322 -5.3465>, rot=<0.5993 -0.2608 -0.2326 -0.7203>}
 upperarm_twist_02_l {parent=upperarm_l, loc=<19.7998 -0.1134 0>, rot=<0 0 0 1>}
 dyn_knot_l {parent=upperarm_twist_02_l, loc=<4.6005 -5.2052 2.0325>, rot=<0.1873 0.0733 0.5778 -0.791>}
 dyn_loop_01_l {parent=dyn_knot_l, loc=<1.0701 -0.2025 1.2624>, rot=<0.6384 0.311 0.2488 -0.6587>}
 dyn_loop_02_l {parent=dyn_knot_l, loc=<1.4013 0.2025 -1.2624>, rot=<0.5888 -0.1041 -0.2903 -0.7471>}
 clavicle_r {parent=spine_05, loc=<5.4831 -0.1053 3.2946>, rot=<0.7289 0.0465 -0.6829 0.009>}
 upperarm_r {parent=clavicle_r, loc=<-9.8658 1.2136 0.4105>, rot=<0.0164 -0.3245 -0.0269 -0.9454>}
 lowerarm_r {parent=upperarm_r, loc=<-26.465 0.8345 -0.259>, rot=<0.0036 0.0161 -0.2298 -0.9731>}
 hand_r {parent=lowerarm_r, loc=<-22.7355 0.3203 0.6106>, rot=<-0.6784 -0.0223 0.0735 -0.7307>}
 thumb_01_r {parent=hand_r, loc=<-1.3118 0.3191 2.459>, rot=<0.6546 -0.4042 0.0664 -0.6354>}
 thumb_02_r {parent=thumb_01_r, loc=<-5.9336 0.0448 -0.0401>, rot=<-0.0037 0.0213 0.298 -0.9543>}
 thumb_03_r {parent=thumb_02_r, loc=<-2.2223 -0.3682 0.0907>, rot=<0.0118 -0.013 0.0066 -0.9998>}
 index_metacarpal_r {parent=hand_r, loc=<-1.6535 -0.0183 2.96>, rot=<0.0158 -0.0124 -0.0173 -0.9996>}
 index_01_r {parent=index_metacarpal_r, loc=<-8.3008 -0.0357 0.0524>, rot=<-0.0044 0.0485 0.0924 -0.9945>}
 index_02_r {parent=index_01_r, loc=<-3.3587 0.3004 -0.0452>, rot=<-0.002 0.0027 0.3225 -0.9466>}
 index_03_r {parent=index_02_r, loc=<-2.5578 0.123 0.0252>, rot=<-0.002 0.0067 0.2558 -0.9667>}
 middle_metacarpal_r {parent=hand_r, loc=<-2.1856 0.1725 0.5984>, rot=<0.006 0.0168 -0.0329 -0.9993>}
 middle_01_r {parent=middle_metacarpal_r, loc=<-7.3275 0.2447 -0.0129>, rot=<-0.0208 0.0042 0.0858 -0.9961>}
 middle_02_r {parent=middle_01_r, loc=<-4.0123 -0.1401 -0.1175>, rot=<-0.0218 0.0258 0.3512 -0.9357>}
 middle_03_r {parent=middle_02_r, loc=<-2.6077 -0.0534 -0.1373>, rot=<0.0082 0.0188 0.1708 -0.9851>}
 ring_metacarpal_r {parent=hand_r, loc=<-2.2074 0.3625 -1.3638>, rot=<-0.0284 0.0722 -0.0458 -0.9959>}
 ring_01_r {parent=ring_metacarpal_r, loc=<-7.0295 0.6197 0.1446>, rot=<-0.0489 -0.0252 0.1892 -0.9804>}
 ring_02_r {parent=ring_01_r, loc=<-3.6174 -0.4647 0.1027>, rot=<0 0 0.2806 -0.9598>}
 ring_03_r {parent=ring_02_r, loc=<-2.4152 -0.114 -0.0199>, rot=<0 0 0.1189 -0.9929>}
 pinky_metacarpal_r {parent=hand_r, loc=<-1.7616 0.9753 -3.3814>, rot=<-0.0397 0.0814 -0.069 -0.9935>}
 pinky_01_r {parent=pinky_metacarpal_r, loc=<-7.1709 0.4831 0.1094>, rot=<-0.0654 -0.0035 0.2242 -0.9723>}
 pinky_02_r {parent=pinky_01_r, loc=<-2.8274 -0.0893 0.1043>, rot=<0.0037 0.0011 0.2854 -0.9584>}
 pinky_03_r {parent=pinky_02_r, loc=<-2.2046 -0.0307 0.0078>, rot=<-0.0002 0.0098 0.1445 -0.9895>}
 deform_wrist_dn_base_r {parent=hand_r, loc=<-0.0001 1.5328 0>, rot=<0 0 0 -1>}
 deform_wrist_dn_r {parent=deform_wrist_dn_base_r, loc=<0 0 0>, rot=<0 0 0 1>}
 deform_wrist_up_base_r {parent=hand_r, loc=<-0.0002 -2.3007 -0.0001>, rot=<0 0 0 -1>}
 deform_wrist_up_r {parent=deform_wrist_up_base_r, loc=<0 0 0>, rot=<0 0 0 1>}
 lowerarm_twist_02_r {parent=lowerarm_r, loc=<-10.6707 0 0>, rot=<0 0 0 -1>}
 lowerarm_twist_01_r {parent=lowerarm_r, loc=<-18.316 0 0>, rot=<0 0 0 -1>}
 dyn_sleeve_lwr_01_r {parent=lowerarm_twist_01_r, loc=<3.9063 2.4053 5.3364>, rot=<-0.3017 -0.3664 0.6873 -0.5499>}
 dyn_sleeve_lwr_02_r {parent=lowerarm_twist_01_r, loc=<4.1535 -1.8002 -5.2352>, rot=<-0.5405 -0.5638 -0.4572 0.4253>}
 deform_elbow_in_base_r {parent=lowerarm_r, loc=<-0.1834 -3.6062 0.2572>, rot=<0 0 0 -1>}
 deform_elbow_in_r {parent=deform_elbow_in_base_r, loc=<0 0 0>, rot=<0 0 0 1>}
 deform_elbow_out_base_r {parent=lowerarm_r, loc=<0 4 -0.0001>, rot=<0 0 0 -1>}
 deform_elbow_out_r {parent=deform_elbow_out_base_r, loc=<0 0 0>, rot=<0 0 0 1>}
 upperarm_twist_01_r {parent=upperarm_r, loc=<-7.2135 0.1134 0>, rot=<0 0 0 -1>}
 deform_upperarm_bk_base_r {parent=upperarm_twist_01_r, loc=<6.4053 4.3746 -0.314>, rot=<0 0 0 -1>}
 deform_upperarm_bk_r {parent=deform_upperarm_bk_base_r, loc=<0 0 0>, rot=<0 0 0 1>}
 deform_upperarm_fr_base_r {parent=upperarm_twist_01_r, loc=<5.6312 -5.092 -0.3354>, rot=<0 0 0 -1>}
 deform_upperarm_fr_r {parent=deform_upperarm_fr_base_r, loc=<0 0 0>, rot=<0 0 0 1>}
 deform_upperarm_up_base_r {parent=upperarm_twist_01_r, loc=<8.891 -0.4825 -3.9517>, rot=<0 0 0 -1>}
 deform_upperarm_up_r {parent=deform_upperarm_up_base_r, loc=<0 0 0>, rot=<0 0 0 1>}
 dyn_sleeve_upr_r {parent=upperarm_twist_01_r, loc=<-5.6854 0.3322 5.3463>, rot=<0.5993 -0.2608 -0.2326 -0.7203>}
 upperarm_twist_02_r {parent=upperarm_r, loc=<-19.7998 0.1134 0>, rot=<0 0 0 -1>}
 dyn_knot_r {parent=upperarm_twist_02_r, loc=<-4.6008 5.2052 -2.0322>, rot=<0.1873 0.0733 0.5778 -0.791>}
 dyn_loop_01_r {parent=dyn_knot_r, loc=<-1.0701 0.2024 -1.2624>, rot=<0.6384 0.311 0.2488 -0.6587>}
 dyn_loop_02_r {parent=dyn_knot_r, loc=<-1.4014 -0.2018 1.262>, rot=<0.5888 -0.1041 -0.2903 -0.7471>}
 deform_pec_aim_base_l {parent=spine_04, loc=<3.2456 9.3309 -1.2237>, rot=<0.0603 -0.4033 0.3318 -0.8507>}
 deform_pec_aim_l {parent=deform_pec_aim_base_l, loc=<0 0 0>, rot=<0 0 0 1>}
 deform_pec_base_l {parent=deform_pec_aim_l, loc=<9.3665 0 0>, rot=<0 0 0 1>}
 deform_pec_l {parent=deform_pec_base_l, loc=<0 0 0>, rot=<0 0 0 1>}
 deform_pec_aim_base_r {parent=spine_04, loc=<3.2452 9.3309 1.2237>, rot=<-0.4033 -0.0603 0.8507 0.3318>}
 deform_pec_aim_r {parent=deform_pec_aim_base_r, loc=<0 0 0>, rot=<0 0 0 1>}
 deform_pec_base_r {parent=deform_pec_aim_r, loc=<-9.3667 -0.0003 -0.0004>, rot=<0 0 0 1>}
 deform_pec_r {parent=deform_pec_base_r, loc=<0 0 0>, rot=<0 0 0 1>}
 thigh_l {parent=pelvis, loc=<-1.0855 -0.2109 -8.7053>, rot=<0.0009 0.0355 0.0155 -0.9993>}
 calf_l {parent=thigh_l, loc=<-42.5284 -0.0019 -0.7297>, rot=<0.0006 -0.0161 -0.0391 -0.9991>}
 foot_l {parent=calf_l, loc=<-46.7401 -2.5014 -0.592>, rot=<0.0427 -0.0238 0.0353 -0.9982>}
 ball_l {parent=foot_l, loc=<-3.5658 10.8066 -0.1544>, rot=<0.0001 -0.0001 -0.7063 -0.7079>}
 calf_twist_02_l {parent=calf_l, loc=<-19.1546 -1.7943 -0.2638>, rot=<0.0237 0.0095 -0.0035 -0.9997>}
 calf_twist_01_l {parent=calf_l, loc=<-35.8483 -1.9197 -0.5794>, rot=<0.0237 0.0095 -0.0035 -0.9997>}
 deform_knee_in_base_l {parent=calf_l, loc=<0.8031 -5.0231 0.2456>, rot=<0 0 0 -1>}
 deform_knee_in_l {parent=deform_knee_in_base_l, loc=<0 0 0>, rot=<0 0 0 1>}
 deform_knee_out_base_l {parent=calf_l, loc=<0 4 0>, rot=<0 0 0 -1>}
 deform_knee_out_l {parent=deform_knee_out_base_l, loc=<0 0 0>, rot=<0 0 0 1>}
 thigh_twist_01_l {parent=thigh_l, loc=<-27.952 0 0>, rot=<0.0013 0 0 -1>}
 dyn_pouch {parent=thigh_twist_01_l, loc=<12.2077 0.9829 -8.2195>, rot=<0.014 0.7059 0.708 -0.0106>}
 dyn_flap_01_l {parent=dyn_pouch, loc=<0 0 0>, rot=<-0.0069 -0.0792 0.9894 -0.1217>}
 deform_groin_base_l {parent=thigh_l, loc=<0.4801 5.6776 -0.2486>, rot=<0 0 0 -1>}
 deform_groin_l {parent=deform_groin_base_l, loc=<0 0 0>, rot=<0 0 0 1>}
 thigh_r {parent=pelvis, loc=<-1.0855 -0.2109 8.7053>, rot=<0.0355 -0.0009 0.9993 0.0155>}
 calf_r {parent=thigh_r, loc=<42.5284 0.0019 0.7297>, rot=<0.0006 -0.0161 -0.0391 -0.9991>}
 foot_r {parent=calf_r, loc=<46.7401 2.5014 0.592>, rot=<0.0427 -0.0238 0.0353 -0.9982>}
 ball_r {parent=foot_r, loc=<3.5658 -10.8066 0.1544>, rot=<0.0001 -0.0001 -0.7063 -0.7079>}
 calf_twist_02_r {parent=calf_r, loc=<19.1546 1.7943 0.2638>, rot=<0.0237 0.0095 -0.0035 -0.9997>}
 calf_twist_01_r {parent=calf_r, loc=<35.8483 1.9197 0.5794>, rot=<0.0237 0.0095 -0.0035 -0.9997>}
 deform_knee_in_base_r {parent=calf_r, loc=<-0.8031 5.0231 -0.2457>, rot=<0 0 0 -1>}
 deform_knee_in_r {parent=deform_knee_in_base_r, loc=<0 0 0>, rot=<0 0 0 1>}
 deform_knee_out_base_r {parent=calf_r, loc=<0 -4 0>, rot=<0 0 0 -1>}
 deform_knee_out_r {parent=deform_knee_out_base_r, loc=<0 0 0>, rot=<0 0 0 1>}
 thigh_twist_01_r {parent=thigh_r, loc=<27.952 0 0>, rot=<0.0013 0 0 -1>}
 deform_groin_base_r {parent=thigh_r, loc=<-0.4801 -5.6776 0.2486>, rot=<0 0 0 -1>}
 deform_groin_r {parent=deform_groin_base_r, loc=<0 0 0>, rot=<0 0 0 1>}
 deform_glute_base_l {parent=pelvis, loc=<-7.3466 -9.0213 -9.1245>, rot=<0.0172 0.031 -0.4477 -0.8935>}
 deform_glute_l {parent=deform_glute_base_l, loc=<0 0 0>, rot=<0 0 0 1>}
 deform_glute_base_r {parent=pelvis, loc=<-7.3464 -9.0213 9.1245>, rot=<0.031 -0.0172 0.8935 -0.4477>}
 deform_glute_r {parent=deform_glute_base_r, loc=<0 0 0>, rot=<0 0 0 1>}
 dyn_tail_parent {parent=pelvis, loc=<-1.6161 -7.0046 -0.0543>, rot=<0.6943 -0.7197 0 0>}
 dyn_tail_A_parent {parent=dyn_tail_parent, loc=<2.2171 -0.0474 3.9955>, rot=<0.0204 0.2168 -0.0914 -0.9717>}
 dyn_mstr_tail_A {parent=dyn_tail_A_parent, loc=<0 0 0>, rot=<0 0 0 1>}
 dyn_tail_A_01 {parent=dyn_mstr_tail_A, loc=<0 0 0>, rot=<0 0 0 1>}
 dyn_tail_A_02 {parent=dyn_tail_A_01, loc=<17.9378 0 0>, rot=<0.0079 0.0296 -0.0723 -0.9969>}
 dyn_tail_A_03 {parent=dyn_tail_A_02, loc=<21.3536 0 0>, rot=<-0.0071 -0.0483 0.1851 -0.9815>}
 dyn_tail_B_parent {parent=dyn_tail_parent, loc=<1.558 -0.0082 4.0652>, rot=<-0.0043 0.2226 0.0189 -0.9747>}
 dyn_mstr_tail_B {parent=dyn_tail_B_parent, loc=<0 0 0>, rot=<0 0 0 1>}
 dyn_tail_B_01 {parent=dyn_mstr_tail_B, loc=<0 0 0>, rot=<0 0 0 1>}
 dyn_tail_B_02 {parent=dyn_tail_B_01, loc=<17.8053 0 0>, rot=<0.0005 0.0166 -0.07 -0.9974>}
 dyn_tail_B_03 {parent=dyn_tail_B_02, loc=<21.5549 0 0>, rot=<0.0019 -0.0217 0.188 -0.9819>}
 dyn_tail_C_parent {parent=dyn_tail_parent, loc=<3.452 -2.0311 0.4112>, rot=<-0.0203 0.1877 0.1057 -0.9763>}
 dyn_mstr_tail_C {parent=dyn_tail_C_parent, loc=<0 0 0>, rot=<0 0 0 1>}
 dyn_tail_C_01 {parent=dyn_mstr_tail_C, loc=<0 0 0>, rot=<0 0 0 1>}
 dyn_tail_C_02 {parent=dyn_tail_C_01, loc=<20.2912 0 0>, rot=<-0.0125 0.0473 0.0434 -0.9979>}
 dyn_tail_C_03 {parent=dyn_tail_C_02, loc=<17.9167 0 0>, rot=<0.0001 0.0022 -0.3381 -0.9411>}
 dyn_tail_D_parent {parent=dyn_tail_parent, loc=<3.5428 -0.2075 -1.0978>, rot=<0.0369 -0.2034 0.1748 -0.9626>}
 dyn_mstr_tail_D {parent=dyn_tail_D_parent, loc=<0 0 0>, rot=<0 0 0 1>}
 dyn_tail_D_01 {parent=dyn_mstr_tail_D, loc=<0 0 0>, rot=<0 0 0 1>}
 dyn_tail_D_02 {parent=dyn_tail_D_01, loc=<17.9036 0 0>, rot=<-0.0108 0.0348 -0.0592 -0.9976>}
 dyn_tail_D_03 {parent=dyn_tail_D_02, loc=<21.5119 0 0>, rot=<0.0392 -0.0891 0.1721 -0.9803>}
 dyn_tail_E_parent {parent=dyn_tail_parent, loc=<2.2129 -0.1948 -2.391>, rot=<-0.0091 -0.2308 -0.0383 -0.9722>}
 dyn_mstr_tail_E {parent=dyn_tail_E_parent, loc=<0 0 0>, rot=<0 0 0 1>}
 dyn_tail_E_01 {parent=dyn_mstr_tail_E, loc=<0 0 0>, rot=<0 0 0 1>}
 dyn_tail_E_02 {parent=dyn_tail_E_01, loc=<14.0443 0 0>, rot=<0.0005 -0.029 0.0953 -0.995>}
 dyn_tail_E_03 {parent=dyn_tail_E_02, loc=<18.1687 0 0>, rot=<-0.0003 0.1704 -0.1083 -0.9794>}
 dyn_tail_F_parent {parent=dyn_tail_parent, loc=<1.9064 -0.5167 -1.1509>, rot=<0.0005 0.0399 -0.0137 -0.9991>}
 dyn_mstr_tail_F {parent=dyn_tail_F_parent, loc=<0 0 0>, rot=<0 0 0 1>}
 dyn_tail_F_01 {parent=dyn_mstr_tail_F, loc=<0 0 0>, rot=<0 0 0 1>}
 dyn_tail_F_02 {parent=dyn_tail_F_01, loc=<17.8594 0 0>, rot=<0.0003 -0.0078 0.0719 -0.9974>}
 dyn_tail_F_03 {parent=dyn_tail_F_02, loc=<21.512 0 0>, rot=<-0.0004 -0.0055 -0.1963 -0.9805>}
 dyn_tail_G_parent {parent=dyn_tail_parent, loc=<3.3927 1.5323 -2.05>, rot=<-0.0385 -0.2491 -0.1479 -0.9563>}
 dyn_mstr_tail_G {parent=dyn_tail_G_parent, loc=<0 0 0>, rot=<0 0 0 1>}
 dyn_tail_G_01 {parent=dyn_mstr_tail_G, loc=<0 0 0>, rot=<0 0 0 1>}
 dyn_tail_G_02 {parent=dyn_tail_G_01, loc=<13.9547 0 0>, rot=<0.0001 0.0006 0.0998 -0.995>}
 dyn_tail_G_03 {parent=dyn_tail_G_02, loc=<18.5073 0 0>, rot=<0.038 0.1439 -0.1492 -0.9775>}
 dyn_tuft_A_constrain_parent {parent=pelvis, loc=<1.3566 -10.1305 0.0347>, rot=<0.9156 -0.402 -0.0011 -0.0029>}
 dyn_tuft_A_parent {parent=dyn_tuft_A_constrain_parent, loc=<0 0 0>, rot=<0 0 0 1>}
 dyn_tuft_A {parent=dyn_tuft_A_parent, loc=<0 0 0>, rot=<0 0 0 1>}
 dyn_tuft_B_constrain_parent {parent=pelvis, loc=<3.613 -8.5037 -7.5082>, rot=<0.9041 -0.2654 0.2563 -0.2157>}
 dyn_tuft_B_parent {parent=dyn_tuft_B_constrain_parent, loc=<0 0 0>, rot=<0 0 0 1>}
 dyn_tuft_B {parent=dyn_tuft_B_parent, loc=<0 0 0>, rot=<0 0 0 1>}
 dyn_tuft_C_constrain_parent {parent=pelvis, loc=<4.6038 -5.3516 -10.4273>, rot=<0.8166 -0.2339 0.3775 -0.3688>}
 dyn_tuft_C_parent {parent=dyn_tuft_C_constrain_parent, loc=<0 0 0>, rot=<0 0 0 1>}
 dyn_tuft_C {parent=dyn_tuft_C_parent, loc=<0 0 0>, rot=<0 0 0 1>}
 dyn_tuft_D_constrain_parent {parent=pelvis, loc=<5.0862 -0.9623 -11.7884>, rot=<0.6959 -0.2571 0.1843 -0.6447>}
 dyn_tuft_D_parent {parent=dyn_tuft_D_constrain_parent, loc=<0 0 0>, rot=<0 0 0 1>}
 dyn_tuft_D {parent=dyn_tuft_D_parent, loc=<0 0 0>, rot=<0 0 0 1>}
 dyn_tuft_E_constrain_parent {parent=pelvis, loc=<3.6126 -8.5037 7.5082>, rot=<-0.2654 -0.9041 0.2157 0.2563>}
 dyn_tuft_E_parent {parent=dyn_tuft_E_constrain_parent, loc=<0 0 0>, rot=<0 0 0 1>}
 dyn_tuft_E {parent=dyn_tuft_E_parent, loc=<0 0 0>, rot=<0 0.0087 1 0>}
 dyn_tuft_F_constrain_parent {parent=pelvis, loc=<4.6041 -5.3516 10.4273>, rot=<-0.2339 -0.8166 0.3688 0.3775>}
 dyn_tuft_F_parent {parent=dyn_tuft_F_constrain_parent, loc=<0 0 0>, rot=<0 0 0 1>}
 dyn_tuft_F {parent=dyn_tuft_F_parent, loc=<0 0 0>, rot=<0 -0.0349 0.9994 0>}
 dyn_tuft_G_constrain_parent {parent=pelvis, loc=<5.0857 -0.9623 11.7884>, rot=<-0.2571 -0.6959 0.6447 0.1843>}
 dyn_tuft_G_parent {parent=dyn_tuft_G_constrain_parent, loc=<0 0 0>, rot=<0 0 0 1>}
 dyn_tuft_G {parent=dyn_tuft_G_parent, loc=<0 0 0>, rot=<0 0 1 0>}
]
