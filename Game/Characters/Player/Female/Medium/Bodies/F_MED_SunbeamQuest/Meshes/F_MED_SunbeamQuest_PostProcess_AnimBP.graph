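# Post-process AnimBP for the female SunbeamQuest player body that controls its seven dynamic tail chains.
# Bone scaling reveals tail sets by AthenaKills stat, cycles them in the front end, or follows cape and variant-stage overrides.
# Spine twist drives additive tail-root offsets, then trail controllers add secondary motion.
# Skydive float or dive blends in constrained bone rotations, and a Control Rig runs when stationary outside skydiving.
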AnimBlueprint /Game/Characters/Player/Female/Medium/Bodies/F_MED_SunbeamQuest/Meshes/F_MED_SunbeamQuest_PostProcess_AnimBP
target_skeleton=&/F_MED_SunbeamQuest_Skeleton
graphs [
 AnimGraph {nodes [
   n1 = Get tailSkydiveAlpha() -> (tailSkydiveAlpha:double)
   n2 = "Transform (Modify) Bone - Bone: dyn_tail_parent"(
    ComponentPose=Local To Component(LocalPose=Copy Pose From Mesh() {bUseAttachedParent=true})
    Alpha=1
    Translation=Get tail_translate() -> (tail_translate:Vector)
    Rotation=Get tail_rotate() -> (tail_rotate:Rotator)
    Scale="1,1,1"
   ) {rotationMode=BMM_Additive, translationMode=BMM_Additive, alphaBoolBlend {blendOption=Linear}, boneToModify {boneName=dyn_tail_parent}}
   n3 = "Transform (Modify) Bone - Bone: dyn_tail_A_parent"(ComponentPose=@n2, Alpha=1, Translation="0,0,0", Rotation="0,0,0", Scale=Get Tail_2_Scale() -> (Tail_2_Scale:Vector)) {
    scaleMode=BMM_Replace
    translationSpace=BCS_BoneSpace
    alphaBoolBlend {blendOption=Linear}
    alphaScaleBiasClamp {bInterpResult=true, interpSpeedDecreasing=5, interpSpeedIncreasing=5}
    boneToModify {boneName=dyn_tail_A_parent}
   }
   n4 = "Transform (Modify) Bone - Bone: dyn_tail_B_parent"(ComponentPose=@n3, Alpha=1, Translation="0,0,0", Rotation="0,0,0", Scale=Get TailPrime_1_Scale() -> (TailPrime_1_Scale:Vector)) {
    scaleMode=BMM_Replace
    alphaBoolBlend {blendOption=Linear}
    alphaScaleBiasClamp {bInterpResult=true, interpSpeedDecreasing=5, interpSpeedIncreasing=5}
    boneToModify {boneName=dyn_tail_B_parent}
   }
   n5 = "Transform (Modify) Bone - Bone: dyn_tail_C_parent"(ComponentPose=@n4, Alpha=1, Translation="0,-2,0", Rotation="0,0,0", Scale=Get Tail_4_Scale() -> (Tail_4_Scale:Vector)) {
    scaleMode=BMM_Replace
    translationMode=BMM_Additive
    translationSpace=BCS_BoneSpace
    alphaBoolBlend {blendOption=Linear}
    alphaScaleBiasClamp {bInterpResult=true, interpSpeedDecreasing=5, interpSpeedIncreasing=5}
    boneToModify {boneName=dyn_tail_C_parent}
    translation {y=-2}
   }
   n6 = "Transform (Modify) Bone - Bone: dyn_tail_D_parent"(ComponentPose=@n5, Alpha=1, Translation="0,2,0", Rotation="0,0,0", Scale=Get Tail_5_Scale() -> (Tail_5_Scale:Vector)) {
    scaleMode=BMM_Replace
    translationMode=BMM_Additive
    translationSpace=BCS_BoneSpace
    alphaBoolBlend {blendOption=Linear}
    alphaScaleBiasClamp {bInterpResult=true, interpSpeedDecreasing=5, interpSpeedIncreasing=5}
    boneToModify {boneName=dyn_tail_D_parent}
    translation {y=2}
   }
   n7 = "Transform (Modify) Bone - Bone: dyn_tail_E_parent"(ComponentPose=@n6, Alpha=1, Translation="0,0,0", Rotation="0,0,0", Scale=Get TailPrime_2_Scale() -> (TailPrime_2_Scale:Vector)) {
    scaleMode=BMM_Replace
    alphaBoolBlend {blendOption=Linear}
    alphaScaleBiasClamp {bInterpResult=true, interpSpeedDecreasing=5, interpSpeedIncreasing=5}
    boneToModify {boneName=dyn_tail_E_parent}
   }
   n8 = "Transform (Modify) Bone - Bone: dyn_tail_F_parent"(ComponentPose=@n7, Alpha=1, Translation="0,0,0", Rotation="0,0,0", Scale=Get Tail_1_Scale() -> (Tail_1_Scale:Vector)) {
    scaleMode=BMM_Replace
    alphaBoolBlend {blendOption=Linear}
    alphaScaleBiasClamp {bInterpResult=true, interpSpeedDecreasing=5, interpSpeedIncreasing=5}
    boneToModify {boneName=dyn_tail_F_parent}
   }
   n9 = "Transform (Modify) Bone - Bone: dyn_tail_G_parent"(ComponentPose=@n8, Alpha=1, Translation="0,2,0", Rotation="0,0,0", Scale=Get Tail_3_Scale() -> (Tail_3_Scale:Vector)) {
    scaleMode=BMM_Replace
    translationMode=BMM_Additive
    translationSpace=BCS_BoneSpace
    alphaBoolBlend {blendOption=Linear}
    alphaScaleBiasClamp {bInterpResult=true, interpSpeedDecreasing=5, interpSpeedIncreasing=5}
    boneToModify {boneName=dyn_tail_G_parent}
    translation {y=2}
   }
   n10 = Trail controller - Bone: dyn_tail_F_03(ComponentPose=@n9, Alpha=@n1) {
    bLimitRotation=true
    chainLength=3
    maxDeltaTime=0.0333
    relaxationSpeedScale=3
    alphaScaleBiasClamp {bInterpResult=true, interpSpeedDecreasing=0, interpSpeedIncreasing=0}
    relaxationSpeedScaleInputProcessor {bInterpResult=true}
    rotationLimits [{limitMax=<20 40 40>, limitMin=<-20 -40 40>}, {limitMax=<20 40 40>, limitMin=<-20 -40 40>}, {limitMax=<20 40 40>, limitMin=<-20 -40 40>}]
    rotationOffsets [<0 0 0>, <0 0 0>, <0 0 0>]
    trailBone {boneName=dyn_tail_F_03}
    trailRelaxationSpeed {editorCurveData {keys [
       {arriveTangent=-10.2937, interpMode=RCIM_Cubic, leaveTangent=-10.2937, tangentMode=RCTM_User, time=0, value=10}
       {interpMode=RCIM_Cubic, tangentMode=RCTM_User, time=1, value=5}
      ]}}
   }
   n11 = Trail controller - Bone: dyn_tail_A_03(ComponentPose=@n10, Alpha=@n1) {
    bLimitRotation=true
    chainLength=3
    maxDeltaTime=0.0333
    relaxationSpeedScale=3
    alphaScaleBiasClamp {bInterpResult=true, interpSpeedDecreasing=0, interpSpeedIncreasing=0}
    relaxationSpeedScaleInputProcessor {bInterpResult=true}
    rotationLimits [{limitMax=<20 40 40>, limitMin=<-20 -40 40>}, {limitMax=<20 40 40>, limitMin=<-20 -40 40>}, {limitMax=<20 40 40>, limitMin=<-20 -40 40>}]
    rotationOffsets [<0 0 0>, <0 0 0>, <0 0 0>]
    trailBone {boneName=dyn_tail_A_03}
    trailRelaxationSpeed {editorCurveData {keys [
       {arriveTangent=-10.2937, interpMode=RCIM_Cubic, leaveTangent=-10.2937, tangentMode=RCTM_User, time=0, value=10}
       {interpMode=RCIM_Cubic, tangentMode=RCTM_User, time=1, value=5}
      ]}}
   }
   n12 = Trail controller - Bone: dyn_tail_B_03(ComponentPose=@n11, Alpha=@n1) {
    bLimitRotation=true
    chainLength=3
    maxDeltaTime=0.0333
    relaxationSpeedScale=3
    alphaScaleBiasClamp {bInterpResult=true, interpSpeedDecreasing=0, interpSpeedIncreasing=0}
    relaxationSpeedScaleInputProcessor {bInterpResult=true}
    rotationLimits [{limitMax=<20 40 40>, limitMin=<-20 -40 40>}, {limitMax=<20 40 40>, limitMin=<-20 -40 40>}, {limitMax=<20 40 40>, limitMin=<-20 -40 40>}]
    rotationOffsets [<0 0 0>, <0 0 0>, <0 0 0>]
    trailBone {boneName=dyn_tail_B_03}
    trailRelaxationSpeed {editorCurveData {keys [
       {arriveTangent=-10.2937, interpMode=RCIM_Cubic, leaveTangent=-10.2937, tangentMode=RCTM_User, time=0, value=10}
       {interpMode=RCIM_Cubic, tangentMode=RCTM_User, time=1, value=5}
      ]}}
   }
   n13 = Trail controller - Bone: dyn_tail_C_03(ComponentPose=@n12, Alpha=@n1) {
    bLimitRotation=true
    chainLength=3
    maxDeltaTime=0.0333
    relaxationSpeedScale=3
    alphaScaleBiasClamp {bInterpResult=true, interpSpeedDecreasing=0, interpSpeedIncreasing=0}
    relaxationSpeedScaleInputProcessor {bInterpResult=true}
    rotationLimits [{limitMax=<20 40 40>, limitMin=<-20 -40 40>}, {limitMax=<20 40 40>, limitMin=<-20 -40 40>}, {limitMax=<20 40 40>, limitMin=<-20 -40 40>}]
    rotationOffsets [<0 0 0>, <0 0 0>, <0 0 0>]
    trailBone {boneName=dyn_tail_C_03}
    trailRelaxationSpeed {editorCurveData {keys [
       {arriveTangent=-10.2937, interpMode=RCIM_Cubic, leaveTangent=-10.2937, tangentMode=RCTM_User, time=0, value=10}
       {interpMode=RCIM_Cubic, tangentMode=RCTM_User, time=1, value=5}
      ]}}
   }
   n14 = Trail controller - Bone: dyn_tail_D_03(ComponentPose=@n13, Alpha=@n1) {
    bLimitRotation=true
    chainLength=3
    maxDeltaTime=0.0333
    relaxationSpeedScale=3
    alphaScaleBiasClamp {bInterpResult=true, interpSpeedDecreasing=0, interpSpeedIncreasing=0}
    relaxationSpeedScaleInputProcessor {bInterpResult=true}
    rotationLimits [{limitMax=<20 40 40>, limitMin=<-20 -40 40>}, {limitMax=<20 40 40>, limitMin=<-20 -40 40>}, {limitMax=<20 40 40>, limitMin=<-20 -40 40>}]
    rotationOffsets [<0 0 0>, <0 0 0>, <0 0 0>]
    trailBone {boneName=dyn_tail_D_03}
    trailRelaxationSpeed {editorCurveData {keys [
       {arriveTangent=-10.2937, interpMode=RCIM_Cubic, leaveTangent=-10.2937, tangentMode=RCTM_User, time=0, value=10}
       {interpMode=RCIM_Cubic, tangentMode=RCTM_User, time=1, value=5}
      ]}}
   }
   n15 = Trail controller - Bone: dyn_tail_E_03(ComponentPose=@n14, Alpha=@n1) {
    bLimitRotation=true
    chainLength=3
    maxDeltaTime=0.0333
    relaxationSpeedScale=3
    alphaScaleBiasClamp {bInterpResult=true, interpSpeedDecreasing=0, interpSpeedIncreasing=0}
    relaxationSpeedScaleInputProcessor {bInterpResult=true}
    rotationLimits [{limitMax=<20 40 40>, limitMin=<-20 -40 40>}, {limitMax=<20 40 40>, limitMin=<-20 -40 40>}, {limitMax=<20 40 40>, limitMin=<-20 -40 40>}]
    rotationOffsets [<0 0 0>, <0 0 0>, <0 0 0>]
    trailBone {boneName=dyn_tail_E_03}
    trailRelaxationSpeed {editorCurveData {keys [
       {arriveTangent=-10.2937, interpMode=RCIM_Cubic, leaveTangent=-10.2937, tangentMode=RCTM_User, time=0, value=10}
       {interpMode=RCIM_Cubic, tangentMode=RCTM_User, time=1, value=5}
      ]}}
   }
   n16 = Trail controller - Bone: dyn_tail_G_03(ComponentPose=@n15, Alpha=@n1) {
    bLimitRotation=true
    chainLength=3
    maxDeltaTime=0.0333
    relaxationSpeedScale=3
    alphaScaleBiasClamp {bInterpResult=true, interpSpeedDecreasing=0, interpSpeedIncreasing=0}
    relaxationSpeedScaleInputProcessor {bInterpResult=true}
    rotationLimits [{limitMax=<20 40 40>, limitMin=<-20 -40 40>}, {limitMax=<20 40 40>, limitMin=<-20 -40 40>}, {limitMax=<20 40 40>, limitMin=<-20 -40 40>}]
    rotationOffsets [<0 0 0>, <0 0 0>, <0 0 0>]
    trailBone {boneName=dyn_tail_G_03}
    trailRelaxationSpeed {editorCurveData {keys [
       {arriveTangent=-10.2937, interpMode=RCIM_Cubic, leaveTangent=-10.2937, tangentMode=RCTM_User, time=0, value=10}
       {interpMode=RCIM_Cubic, tangentMode=RCTM_User, time=1, value=5}
      ]}}
   }
   n17 = Get limitTailRange() -> (limitTailRange:bool)
   n18 = "Transform (Modify) Bone - Bone: dyn_tail_parent"!(ComponentPose=@n16, bAlphaBoolEnabled:bool=@n17, Translation="0,-6,0", Rotation="0,0,0", Scale="1,1,1") -> (Pose) {
    alphaInputType=Bool
    rotationSpace=BCS_BoneSpace
    translationMode=BMM_Additive
    alphaBoolBlend {blendInTime=4, blendOption=Linear, blendOutTime=4}
    boneToModify {boneName=dyn_tail_parent}
    translation {y=-6}
   }
   n19 = Get tail_rotate()
   n20 = "Transform (Modify) Bone - Bone: dyn_tail_A_01"!(ComponentPose=@n18, bAlphaBoolEnabled:bool=@n17, Translation="0,0,0", Rotation=@n19, Scale="1,1,1") -> (Pose) {
    alphaInputType=Bool
    rotationMode=BMM_Replace
    rotationSpace=BCS_BoneSpace
    alphaBoolBlend {blendInTime=4, blendOption=Linear, blendOutTime=4}
    boneToModify {boneName=dyn_tail_A_01}
   }
   n21 = "Transform (Modify) Bone - Bone: dyn_tail_B_01"!(ComponentPose=@n20, bAlphaBoolEnabled:bool=@n17, Translation="0,0,0", Rotation=@n19, Scale="1,1,1") -> (Pose) {
    alphaInputType=Bool
    rotationMode=BMM_Replace
    rotationSpace=BCS_BoneSpace
    alphaBoolBlend {blendInTime=4, blendOption=Linear, blendOutTime=4}
    boneToModify {boneName=dyn_tail_B_01}
   }
   n22 = "Transform (Modify) Bone - Bone: dyn_tail_C_01"!(ComponentPose=@n21, bAlphaBoolEnabled:bool=@n17, Translation="0,0,0", Rotation=@n19, Scale="1,1,1") -> (Pose) {
    alphaInputType=Bool
    rotationMode=BMM_Replace
    rotationSpace=BCS_BoneSpace
    alphaBoolBlend {blendInTime=4, blendOption=Linear, blendOutTime=4}
    boneToModify {boneName=dyn_tail_C_01}
   }
   n23 = "Transform (Modify) Bone - Bone: dyn_tail_D_01"!(ComponentPose=@n22, bAlphaBoolEnabled:bool=@n17, Translation="0,0,0", Rotation=@n19, Scale="1,1,1") -> (Pose) {
    alphaInputType=Bool
    rotationMode=BMM_Replace
    rotationSpace=BCS_BoneSpace
    alphaBoolBlend {blendInTime=4, blendOption=Linear, blendOutTime=4}
    boneToModify {boneName=dyn_tail_D_01}
   }
   n24 = "Transform (Modify) Bone - Bone: dyn_tail_E_01"!(ComponentPose=@n23, bAlphaBoolEnabled:bool=@n17, Translation="0,0,0", Rotation=@n19, Scale="1,1,1") -> (Pose) {
    alphaInputType=Bool
    rotationMode=BMM_Replace
    rotationSpace=BCS_BoneSpace
    alphaBoolBlend {blendInTime=4, blendOption=Linear, blendOutTime=4}
    boneToModify {boneName=dyn_tail_E_01}
   }
   n25 = "Transform (Modify) Bone - Bone: dyn_tail_F_01"!(ComponentPose=@n24, bAlphaBoolEnabled:bool=@n17, Translation="0,0,0", Rotation="0,0,0", Scale="1,1,1") -> (Pose) {
    alphaInputType=Bool
    rotationMode=BMM_Replace
    rotationSpace=BCS_BoneSpace
    alphaBoolBlend {blendInTime=4, blendOption=Linear, blendOutTime=4}
    boneToModify {boneName=dyn_tail_F_01}
   }
   n26 = "Transform (Modify) Bone - Bone: dyn_tail_G_01"!(ComponentPose=@n25, bAlphaBoolEnabled:bool=@n17, Translation="0,0,0", Rotation=@n19, Scale="1,1,1") -> (Pose) {
    alphaInputType=Bool
    rotationMode=BMM_Replace
    rotationSpace=BCS_BoneSpace
    alphaBoolBlend {blendInTime=4, blendOption=Linear, blendOutTime=4}
    boneToModify {boneName=dyn_tail_G_01}
   }
   n27 = "Transform (Modify) Bone - Bone: dyn_tail_A_02"!(ComponentPose=@n26, bAlphaBoolEnabled:bool=@n17, Translation="0,0,0", Rotation=@n19, Scale="1,1,1") -> (Pose) {
    alphaInputType=Bool
    rotationMode=BMM_Replace
    rotationSpace=BCS_BoneSpace
    alphaBoolBlend {blendInTime=4, blendOption=Linear, blendOutTime=4}
    boneToModify {boneName=dyn_tail_A_02}
   }
   n28 = "Transform (Modify) Bone - Bone: dyn_tail_B_02"!(ComponentPose=@n27, bAlphaBoolEnabled:bool=@n17, Translation="0,0,0", Rotation=@n19, Scale="1,1,1") -> (Pose) {
    alphaInputType=Bool
    rotationMode=BMM_Replace
    rotationSpace=BCS_BoneSpace
    alphaBoolBlend {blendInTime=4, blendOption=Linear, blendOutTime=4}
    boneToModify {boneName=dyn_tail_B_02}
   }
   n29 = "Transform (Modify) Bone - Bone: dyn_tail_C_02"!(ComponentPose=@n28, bAlphaBoolEnabled:bool=@n17, Translation="0,0,0", Rotation=@n19, Scale="1,1,1") -> (Pose) {
    alphaInputType=Bool
    rotationMode=BMM_Replace
    rotationSpace=BCS_BoneSpace
    alphaBoolBlend {blendInTime=4, blendOption=Linear, blendOutTime=4}
    boneToModify {boneName=dyn_tail_C_02}
   }
   n30 = "Transform (Modify) Bone - Bone: dyn_tail_D_02"!(ComponentPose=@n29, bAlphaBoolEnabled:bool=@n17, Translation="0,0,0", Rotation=@n19, Scale="1,1,1") -> (Pose) {
    alphaInputType=Bool
    rotationMode=BMM_Replace
    rotationSpace=BCS_BoneSpace
    alphaBoolBlend {blendInTime=4, blendOption=Linear, blendOutTime=4}
    boneToModify {boneName=dyn_tail_D_02}
   }
   n31 = "Transform (Modify) Bone - Bone: dyn_tail_E_02"!(ComponentPose=@n30, bAlphaBoolEnabled:bool=@n17, Translation="0,0,0", Rotation=@n19, Scale="1,1,1") -> (Pose) {
    alphaInputType=Bool
    rotationMode=BMM_Replace
    rotationSpace=BCS_BoneSpace
    alphaBoolBlend {blendInTime=4, blendOption=Linear, blendOutTime=4}
    boneToModify {boneName=dyn_tail_E_02}
   }
   n32 = "Transform (Modify) Bone - Bone: dyn_tail_F_02"!(ComponentPose=@n31, bAlphaBoolEnabled:bool=@n17, Translation="0,0,0", Rotation=@n19, Scale="1,1,1") -> (Pose) {
    alphaInputType=Bool
    rotationMode=BMM_Replace
    rotationSpace=BCS_BoneSpace
    alphaBoolBlend {blendInTime=4, blendOption=Linear, blendOutTime=4}
    boneToModify {boneName=dyn_tail_F_02}
   }
   n33 = "Transform (Modify) Bone - Bone: dyn_tail_G_02"!(ComponentPose=@n32, bAlphaBoolEnabled:bool=@n17, Translation="0,0,0", Rotation=@n19, Scale="1,1,1") -> (Pose) {
    alphaInputType=Bool
    rotationMode=BMM_Replace
    rotationSpace=BCS_BoneSpace
    alphaBoolBlend {blendInTime=4, blendOption=Linear, blendOutTime=4}
    boneToModify {boneName=dyn_tail_G_02}
   }
   n34 = Control Rig!(Source=Component To Local(ComponentPose=@n33), bAlphaBoolEnabled:bool=Get useTailCtrlRig() -> (useTailCtrlRig:bool)) -> (Pose) {
    alphaInputType=Bool
    alphaBoolBlend {blendInTime=0.25, blendOption=Linear, blendOutTime=1}
    alphaScaleBiasClamp {bInterpResult=true}
    controlRigAssetReference {blueprintRigClass=&/F_MED_SunbeamQuest_CtrlRig.F_MED_SunbeamQuest_CtrlRig_C}
   }
   Output Pose(Result=@n34) {layerGroup=None}
  ]}
 EventGraph {nodes [
   Event BlueprintUpdateAnimation()
   > n1 = Sequence!(execute) -> (then_0, then_1, then_2:Exec, then_3:Exec, then_4:Exec, then_5:Exec)
   > Set cape(cape:bool=CheckForBackbling().bIsBackblingACape) -> (Output_Get:bool)
   n2 = Branch(execute=@n1.then_2, Condition=Get bIsSkyDiving())
   > Set useTailCtrlRig(useTailCtrlRig:bool=false) -> (Output_Get:bool)
   > Set tailSkydiveAlpha(tailSkydiveAlpha:double=0.5) -> (Output_Get:double)
   n3 = Sequence!(execute=@n2.else) -> (then_0, then_1)
   > Set tailSkydiveAlpha(tailSkydiveAlpha=1)
   n4 = Branch(execute=@n3.then_1, Condition=Get bIsMoving2D())
   > Set useTailCtrlRig(useTailCtrlRig=false)
   Set useTailCtrlRig(execute=@n4.else, useTailCtrlRig=true)
   n5 = GetOwnerSkeletalMeshComponent()
   n6 = BreakRotator(InRot=BreakTransform(InTransform=GetDeltaTransformFromRefPose(self=@n5, BoneName=spine_05, BaseName=spine_01)).Rotation)
   n7 = MapRangeClamped(Value=@n6.Yaw, InRangeA=19, InRangeB=-19, OutRangeA=-1)
   Set tail_rotate(execute=@n1.then_3, tail_rotate:Rotator=MakeRotator(Roll=@n7, Pitch=@n7, Yaw=@n7)) -> (Output_Get:Rotator)
   > Set tail_translate(tail_translate:Vector=MakeVector(Y=MapRangeClamped(Value=@n6.Yaw, InRangeA=26, InRangeB=-26, OutRangeB=0))) -> (Output_Get:Vector)
   Set limitTailRange(execute=@n1.then_4, limitTailRange:bool=OR Boolean(A=Get bIsSkydiveFloating(), B=Get bIsSkydiveDiveMode())) -> (Output_Get:bool)
   Is Valid(exec=@n1.then_1, InputObject=Get skelMesh() -> (skelMesh:Skeletal Mesh Component Object Reference))
   > n8 = Branch(Condition=Get cape() -> (cape:bool))
   n9 = Sequence!(execute=[@n17.then, @n8.then, @n10.then]) -> (then_0, then_1, then_2:Exec, then_3:Exec, then_4:Exec, then_5:Exec, then_6:Exec, then_7:Exec)
   > Set Tail_2_Scale(Tail_2_Scale:Vector="0, 0, 0") -> (Output_Get:Vector)
   Set TailPrime_1_Scale(execute=@n9.then_1, TailPrime_1_Scale:Vector="0, 0, 0") -> (Output_Get:Vector)
   Set Tail_4_Scale(execute=@n9.then_2, Tail_4_Scale:Vector="0, 0, 0") -> (Output_Get:Vector)
   Set Tail_5_Scale(execute=@n9.then_3, Tail_5_Scale:Vector="0, 0, 0") -> (Output_Get:Vector)
   Set TailPrime_2_Scale(execute=@n9.then_4, TailPrime_2_Scale:Vector="0, 0, 0") -> (Output_Get:Vector)
   Set Tail_1_Scale(execute=@n9.then_5, Tail_1_Scale:Vector="0, 0, 0") -> (Output_Get:Vector)
   Set Tail_3_Scale(execute=@n9.then_6, Tail_3_Scale:Vector="0, 0, 0") -> (Output_Get:Vector)
   n10 = Branch(execute=@n8.else, Condition=CharacterItemDefinitionHasAnyMetaTag(Tags=Make Array("[0]"=(TagName="Cosmetics.Variant.Property.Stage2"))))
   n11 = Branch(execute=@n10.else, Condition=CharacterItemDefinitionHasAnyMetaTag(Tags=Make Array("[0]"=(TagName="Cosmetics.Variant.Property.Stage3"))))
   > n12 = Sequence()
   > Set Tail_2_Scale(Tail_2_Scale="1,1,1")
   Set TailPrime_1_Scale(execute=@n12.then_1, TailPrime_1_Scale="0, 0, 0")
   Set Tail_4_Scale(execute=@n12.then_2, Tail_4_Scale="1,1,1")
   Set Tail_5_Scale(execute=@n12.then_3, Tail_5_Scale="1,1,1")
   Set TailPrime_2_Scale(execute=@n12.then_4, TailPrime_2_Scale="0, 0, 0")
   Set Tail_1_Scale(execute=@n12.then_5, Tail_1_Scale="1,1,1")
   Set Tail_3_Scale(execute=@n12.then_6, Tail_3_Scale="1,1,1")
   n13 = Branch(execute=@n11.else, Condition=Get bIsInFrontEnd())
   > n14 = Branch(Condition=Get bIsPlayingEmote())
   Set activeKill(execute=@n14.else, activeKill:bool=true) -> (Output_Get:bool)
   > Delay(Duration=2)
   > n15 = Set killCount(killCount:double=float + float(A=Get killCount() -> (killCount:double), B=1)) -> (Output_Get:double)
   > n16 = Branch(Condition="float > float"(A=@n15, B=6))
   > n17 = Set killCount(killCount=0)
   Branch(execute=[@n13.else, @n16.else], Condition=Get activeKill() -> (activeKill:bool))
   > n18 = Sequence!(execute) -> (then_0, then_1, then_2, then_3, then_4, then_5)
   > Branch(Condition="Equal (Float)"(A=Get killCount(), B=1))
   > Set Tail_1_Scale(Tail_1_Scale="1,1,1")
   > Set TailPrime_1_Scale(TailPrime_1_Scale="0, 0, 0")
   > Set TailPrime_2_Scale(TailPrime_2_Scale="0, 0, 0")
   > Set Tail_2_Scale(Tail_2_Scale="0, 0, 0")
   > Set Tail_3_Scale(Tail_3_Scale="0, 0, 0")
   > Set Tail_4_Scale(Tail_4_Scale="0, 0, 0")
   > n19 = Set Tail_5_Scale(Tail_5_Scale="0, 0, 0")
   Set activeKill(execute=[@n19.then, @n23.then, @n22.then, @n21.then, @n20.then], activeKill=false)
   Branch(execute=@n18.then_1, Condition="Equal (Float)"(A=Get killCount(), B=2))
   > Set Tail_1_Scale(Tail_1_Scale="0, 0, 0")
   > Set TailPrime_1_Scale(TailPrime_1_Scale="1,1,1")
   > Set TailPrime_2_Scale(TailPrime_2_Scale="1,1,1")
   > Set Tail_2_Scale(Tail_2_Scale="0, 0, 0")
   > Set Tail_3_Scale(Tail_3_Scale="0, 0, 0")
   > Set Tail_4_Scale(Tail_4_Scale="0, 0, 0")
   > n20 = Set Tail_5_Scale(Tail_5_Scale="0, 0, 0")
   Branch(execute=@n18.then_2, Condition="Equal (Float)"(A=Get killCount(), B=3))
   > Set Tail_1_Scale(Tail_1_Scale="1,1,1")
   > Set TailPrime_1_Scale(TailPrime_1_Scale="0, 0, 0")
   > Set TailPrime_2_Scale(TailPrime_2_Scale="0, 0, 0")
   > Set Tail_2_Scale(Tail_2_Scale="1,1,1")
   > Set Tail_3_Scale(Tail_3_Scale="1,1,1")
   > Set Tail_4_Scale(Tail_4_Scale="0, 0, 0")
   > n21 = Set Tail_5_Scale(Tail_5_Scale="0, 0, 0")
   Branch(execute=@n18.then_3, Condition="Equal (Float)"(A=Get killCount(), B=4))
   > Set Tail_1_Scale(Tail_1_Scale="1,1,1")
   > Set TailPrime_1_Scale(TailPrime_1_Scale="0, 0, 0")
   > Set TailPrime_2_Scale(TailPrime_2_Scale="0, 0, 0")
   > Set Tail_2_Scale(Tail_2_Scale="1,1,1")
   > Set Tail_3_Scale(Tail_3_Scale="1,1,1")
   > Set Tail_4_Scale(Tail_4_Scale="1,1,1")
   > n22 = Set Tail_5_Scale(Tail_5_Scale="0, 0, 0")
   Branch(execute=@n18.then_4, Condition="float >= float"(A=Get killCount(), B=5))
   > Set Tail_1_Scale(Tail_1_Scale="1,1,1")
   > Set TailPrime_1_Scale(TailPrime_1_Scale="0, 0, 0")
   > Set TailPrime_2_Scale(TailPrime_2_Scale="0, 0, 0")
   > Set Tail_2_Scale(Tail_2_Scale="1,1,1")
   > Set Tail_3_Scale(Tail_3_Scale="1,1,1")
   > Set Tail_4_Scale(Tail_4_Scale="1,1,1")
   > n23 = Set Tail_5_Scale(Tail_5_Scale="1,1,1")
   Event BlueprintInitializeAnimation()
   > Set SkelMesh(SkelMesh=GetOwnerSkeletalMeshComponent())
   > n24 = Cast To FortPlayerPawn(Object=TryGetFortPawn())
   > n25 = WatchCosmeticStat!(
    execute:Exec
    Pawn:Fort Player Pawn Object Reference=@n24
    StatTag:Fort Stat Manager Tag Structure=(TagName="StatManager.AthenaKills")
    DefaultValue:int=0
    bBroadcastInitialValue:bool=true
    InitialValueBroadcastDelay:float=-1
    ChangeValueBroadcastDelay:float=-1
    bConsumeWaitingNotifications:bool=true
   ) -> (then:Exec, OnStatChanged:Exec, StatName:Name, StatValue:int) {is=title K2Node_AsyncAction}
   Set killCount(execute=@n25.OnStatChanged, killCount="To Float (Integer)"(InInt=@n25.StatValue))
   > Set activeKill(activeKill=true)
   Event OnCharacterCustomizationLoaded()
  ]}
]
variables [
 activeKill {type=boolean}
 cape {type=boolean}
 killCount {type=number}
 limitTailRange {type=boolean}
 SkelMesh {type=SkeletalMeshComponent}
 Tail_1_Scale {type=Vector}
 Tail_2_Scale {type=Vector}
 Tail_3_Scale {type=Vector}
 Tail_4_Scale {type=Vector}
 Tail_5_Scale {type=Vector}
 tail_rotate {type=Rotator}
 tail_translate {type=Vector}
 TailPrime_1_Scale {type=Vector}
 TailPrime_2_Scale {type=Vector}
 tailSkydiveAlpha {type=number}
 useTailCtrlRig {type=boolean}
]
